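# Physics asset for the collar of the F_MED_PizzaParty female player cosmetic body, set up for secondary motion of its dyn_collar bones.
# Seven collar segments across front, B and back chains each pair a tiny MSTR anchor capsule with a larger capsule, constrained with moderate swing and narrow twist limits.
# Head, neck, both clavicles and the right forearm add body capsules around the collar.
# The only non-collar constraint links right forearm to right clavicle with free angular motion.
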PhysicsAsset /BRCosmetics/Characters/Player/Female/Medium/Bodies/F_MED_PizzaParty/Meshes/F_MED_PizzaParty_Collar_Physics
bodies {
 clavicle_l [{shapeType=Capsule, length=4, radius=3, shapeName=clavicle_r_capsule, center=<13.5906 0 -2.1046>, rotation=<rot 65 -0.001 -0.0011>}]
 clavicle_r [{shapeType=Capsule, length=8.5, radius=5, shapeName=clavicle_r_capsule, center=<-13.4041 0 2.4936>, rotation=<rot 50 -0.0004 -0.0006>}]
 dyn_collarB_01_L [{shapeType=Capsule, length=4, radius=2.5, shapeName=dyn_collar_bk_01_capsule, center=<6.325 2.2143 0>, rotation=<rot -90 180 90>}]
 dyn_collarB_01_R [{
   shapeType=Capsule
   length=3
   radius=3.5
   shapeName=dyn_collar_bk_01_capsule
   center=<-4.6509 -1.3891 -0.0577>
   rotation=<rot 0.0002 0.0003 81.5623>
  }]
 dyn_collarB_MSTR_L [{shapeType=Capsule, length=0, radius=0.5, shapeName=dyn_collar_MSTR_L_capsule, rotation=<rot -80.0004 179.9982 -179.9967>}]
 dyn_collarB_MSTR_R [{shapeType=Capsule, length=0, radius=0.5, shapeName=dyn_collar_MSTR_R_capsule, rotation=<rot -80.0004 -179.9982 179.9967>}]
 dyn_collar_01_L [{shapeType=Capsule, length=4.5, radius=2, shapeName=dyn_collar_bk_01_capsule, center=<2.8052 0 0>, rotation=<rot -90 0 -90>}]
 dyn_collar_01_R [{
   shapeType=Capsule
   length=6.5
   radius=2
   shapeName=dyn_collar_bk_01_capsule
   center=<-3.3167 0.1704 0.1588>
   rotation=<rot -85 -89.9982 -0.002>
  }]
 dyn_collar_MSTR_L [{shapeType=Capsule, length=0, radius=0.5, shapeName=dyn_collar_MSTR_L_capsule, rotation=<rot -80.0004 179.9982 -179.9967>}]
 dyn_collar_MSTR_R [{shapeType=Capsule, length=0, radius=0.5, shapeName=dyn_collar_MSTR_R_capsule, rotation=<rot -80.0004 -179.9982 179.9967>}]
 dyn_collar_bk_01 [{shapeType=Capsule, length=7.5, radius=2, shapeName=dyn_collar_bk_01_capsule, center=<2.6764 0 0.6777>, rotation=<rot -90 180 90>}]
 dyn_collar_bk_01_L [{shapeType=Capsule, length=4, radius=2.5, shapeName=dyn_collar_bk_01_capsule, center=<2.2915 0 0>, rotation=<rot -90 180 90>}]
 dyn_collar_bk_01_R [{shapeType=Capsule, length=5.5, radius=2, shapeName=dyn_collar_bk_01_capsule, center=<-2.4655 1.9269 0>, rotation=<rot -90 180 80>}]
 dyn_collar_bk_MSTR [{shapeType=Capsule, length=0, radius=0.5, shapeName=dyn_collar_bk_MSTR_capsule, rotation=<rot -90 2.7126 -2.7126>}]
 dyn_collar_bk_MSTR_L [{shapeType=Capsule, length=0, radius=0.5, shapeName=dyn_collar_bk_MSTR_L_capsule, rotation=<rot -90 36.6528 -36.6522>}]
 dyn_collar_bk_MSTR_R [{shapeType=Capsule, length=0, radius=0.5, shapeName=dyn_collar_bk_MSTR_R_capsule, rotation=<rot -90 -36.6528 36.6523>}]
 head [{shapeType=Capsule, length=6, radius=5, shapeName=head_capsule, center=<0.7657 4.9096 0>, rotation=<rot 90 -179.9736 -179.9734>}]
 lowerarm_r [{
   shapeType=Capsule
   length=19.5
   radius=7.505
   shapeName=lowerarm_r_capsule
   center=<-11.2647 -0.0719 0.0714>
   rotation=<rot -0.0002 -89.9999 92.8124>
  }]
 neck_01 [
  {shapeType=Capsule, length=7, radius=4.5, shapeName=head_capsule, center=<8.4973 -4.8824 0>, rotation=<rot 0 179.9995 180>}
  {shapeType=Capsule, length=11.5, radius=4, shapeName=neck_01_capsule, center=<1.7075 0 0>, rotation=<rot -90 -0.0001 -0.0001>}
 ]
}
constraints [
 {bone1Name=dyn_collarB_MSTR_L, bone2Name=dyn_collarB_01_L, swing1LimitDegrees=45, swing2LimitDegrees=45, twistLimitDegrees=5}
 {bone1Name=dyn_collarB_MSTR_R, bone2Name=dyn_collarB_01_R, swing1LimitDegrees=45, swing2LimitDegrees=45, twistLimitDegrees=5}
 {bone1Name=dyn_collar_MSTR_L, bone2Name=dyn_collar_01_L, swing1LimitDegrees=45, swing2LimitDegrees=45, twistLimitDegrees=5}
 {bone1Name=dyn_collar_MSTR_R, bone2Name=dyn_collar_01_R, swing1LimitDegrees=45, swing2LimitDegrees=45, twistLimitDegrees=5}
 {bone1Name=dyn_collar_bk_MSTR, bone2Name=dyn_collar_bk_01, swing1LimitDegrees=45, swing2LimitDegrees=45, twistLimitDegrees=5}
 {bone1Name=dyn_collar_bk_MSTR_L, bone2Name=dyn_collar_bk_01_L, swing1LimitDegrees=45, swing2LimitDegrees=45, twistLimitDegrees=5}
 {bone1Name=dyn_collar_bk_MSTR_R, bone2Name=dyn_collar_bk_01_R, swing1LimitDegrees=45, swing2LimitDegrees=45, twistLimitDegrees=5}
 {
  bone1Name=lowerarm_r
  bone2Name=clavicle_r
  swing1LimitDegrees=45
  swing1Motion=Free
  swing2LimitDegrees=45
  swing2Motion=Free
  twistLimitDegrees=45
  twistMotion=Free
 }
]
