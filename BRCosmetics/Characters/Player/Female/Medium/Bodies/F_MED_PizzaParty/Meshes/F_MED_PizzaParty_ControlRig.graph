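# Control Rig for the F_MED_PizzaParty female body that adds corrective deformation and steers its jacket and collar bones.
# Setup caches bind and thigh poses for jacket aim controls.
# Forward solve runs the shared DeformRig function, blends those controls by thigh alphas (up alphas come from root attributes), then aims seven jacket chain master bones at their controls.
# Right-hand distance to the collar drives a curve-eased offset on dyn_collarB_MSTR_R.
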ControlRigBlueprint /BRCosmetics/Characters/Player/Female/Medium/Bodies/F_MED_PizzaParty/Meshes/F_MED_PizzaParty_ControlRig
graphs [
 New Function_ContainedGraph {nodes [
   n1 = RigVMFunctionEntryNode()
   SequenceAggregate = &RigVMFunction_Sequence(ExecuteContext=@n1.ExecuteContext)
   Subtract = &RigVMFunction_MathDoubleSub(A=1, B=@n1.pose_alpha)
   /Script/ControlRig.RigUnit_HierarchySetPoseItemArray(ElementType=All, Pose=@n1.bind_pose, ItemsToSet=@n1.items, ExecutePin=@SequenceAggregate.A, Weight=@Subtract.Result) {is=struct RigVMUnitNode}
   &RigUnit_HierarchySetPoseItemArray(ElementType=All, ItemsToSet=@n1.items, Weight=@n1.pose_alpha, Pose=@n1.target_pose, ExecutePin=@SequenceAggregate.B)
   > RigVMFunctionReturnNode()
  ]}
 RigVMFunctionLibrary {nodes [set_ctrl_pose = RigVMCollapseNode(ExecuteContext=(), bind_pose=(), items=(), target_pose=())]}
 RigVMModel {nodes [
   RigUnit_BeginExecution = &RigUnit_BeginExecution()
   SequenceAggregate_3 = RigVMAggregateNode(ExecuteContext=@RigUnit_BeginExecution.ExecutePin)
   RigVMFunction_MathQuaternionFromEuler = &RigVMFunction_MathQuaternionFromEuler(Euler=(X=0,Y=0,Z=0))
   FN_DeformRig_FNC_1 = RigVMFunctionReferenceNode(
    DeformRig_ Curve="(deformRig_l_upperarm_up_alpha_72_#1=False,deformRig_r_upperarm_up_alpha_73_#2=False,deformRig_l_upperarm_fr_alpha_74_#3=False,deformRig_r_upperarm_fr_alpha_75_CB51B8 ...(truncated)"
    DeformRig_Bool="(upperarm_1_#4=True,elbow_3_#5=True,wrist_7_#6=True,groin_8_#7=True,glute_17_#8=True,knee_9_40C408324CA261 ...(truncated)"
    DeformRig_Debug=(upperarm_debug_18_#4=False,wrist_debug_20_#6=False,lat_debug_26_#9=False,groinGlute_debug_27_#7=False)
    DeformRig_Value="(upperarm_up_18_#10=(Rotation=(X=0.0,Y=-0.0,Z=0.0,W=1.0),Translation=(X=0.0,Y=0.0,Z=8.0),Scale3D=(X=0.70,Y=1.0,Z=1.0)),upperarm_fwd_fr_19_#11=(Rotation=(X=0.0,Y=-0.0,Z=0.0,W=1.0),Translation=(X=0.0, ...(truncated)"
    DeformRig_Value.elbow_in_24_#12.Rotation=@RigVMFunction_MathQuaternionFromEuler.Result
    DeformRig_Value.elbow_out_26_#13.Rotation=&RigVMFunction_MathQuaternionFromEuler(Euler=(X=0,Y=0,Z=0)).Result
    ExecuteContext=@SequenceAggregate_3.A
   )
   > RigVMVariableNode(Value=RigVMDispatchNode(BoneName=root, Default=0, Name=deformRig_l_thigh_up_alpha) {template="DISPATCH_RigDispatch_GetAnimAttribute(in Name,in BoneName,in Default,out Value,out Found)"}.Value) {variable=deformRig_l_thigh_up_alpha}
   > RigVMVariableNode(Value=RigVMDispatchNode(BoneName=root, Default=0, Name=deformRig_r_thigh_up_alpha) {template="DISPATCH_RigDispatch_GetAnimAttribute(in Name,in BoneName,in Default,out Value,out Found)"}.Value) {variable=deformRig_r_thigh_up_alpha}
   n1 = RigVMVariableNode() {variable=l_thigh_aim_at_ctrls}
   set_ctrl_pose = RigVMFunctionReferenceNode(
    ExecuteContext=@SequenceAggregate_3.B
    pose_alpha=RigVMVariableNode() {variable=deformRig_l_thigh_up_alpha}.Value
    items=@n1.Value
    bind_pose=RigVMVariableNode() {variable=l_thigh_aim_at_ctrls_bind_pose}.Value
    target_pose=RigVMVariableNode() {variable=l_thigh_aim_at_ctrls_thigh_up_pose}.Value
   )
   > set_ctrl_pose_1 = RigVMFunctionReferenceNode(
    bind_pose=/Script/ControlRig.RigUnit_HierarchyGetPoseItemArray(ElementType=All, Initial=False, ItemsToGet=RigVMVariableNode() {variable=l_thigh_aim_at_ctrls}.Value) {is=struct RigVMUnitNode}.Pose
    target_pose=RigVMVariableNode() {variable=l_thigh_aim_at_ctrls_thigh_side_pose}.Value
    items=@n1.Value
    pose_alpha=RigVMVariableNode() {variable=spr_l_thigh_side_alpha}.Value
   )
   n2 = RigVMVariableNode() {variable=r_thigh_aim_at_ctrls}
   > set_ctrl_pose_2 = RigVMFunctionReferenceNode(
    pose_alpha=RigVMVariableNode() {variable=deformRig_r_thigh_up_alpha}.Value
    bind_pose=RigVMVariableNode() {variable=r_thigh_aim_at_ctrls_bind_pose}.Value
    items=@n2.Value
    target_pose=RigVMVariableNode() {variable=r_thigh_aim_at_ctrls_thigh_up_pose}.Value
   )
   > set_ctrl_pose_3 = RigVMFunctionReferenceNode(
    bind_pose=&RigUnit_HierarchyGetPoseItemArray(ElementType=All, Initial=False, ItemsToGet=RigVMVariableNode() {variable=r_thigh_aim_at_ctrls}.Value).Pose
    items=@n2.Value
    target_pose=RigVMVariableNode() {variable=r_thigh_aim_at_ctrls_thigh_side_pose}.Value
    pose_alpha=RigVMVariableNode() {variable=spr_r_thigh_side_alpha}.Value
   )
   &RigUnit_AimConstraintLocalSpaceOffset(
    Child=(Type=Bone,Name="master_L_front_chain_1_jnt")
    Filter=(bX=True,bY=True,bZ=True)
    Parents=((Item=(Type=Control,Name="thigh_l_jacket_front_aim_ctrl"),Weight=1))
    Weight=1
    WorldUp=(Target=(X=0,Y=0,Z=1),Kind=Location,Space=(Type=Bone,Name="pelvis"))
    ExecutePin=@SequenceAggregate_3.C
   )
   > &RigUnit_AimConstraintLocalSpaceOffset(
    Child=(Type=Bone,Name="master_L_side_front_chain_1_jnt")
    Filter=(bX=True,bY=True,bZ=True)
    Parents=((Item=(Type=Control,Name="thigh_l_jacket_side_aim_ctrl"),Weight=1))
    Weight=1
    WorldUp=(Target=(X=0,Y=0,Z=1),Kind=Location,Space=(Type=Bone,Name="pelvis"))
   )
   > &RigUnit_AimConstraintLocalSpaceOffset(
    Child=(Type=Bone,Name="master_L_side_rear_chain_1_jnt")
    Filter=(bX=True,bY=True,bZ=True)
    Parents=((Item=(Type=Control,Name="thigh_l_jacket_side_rear_aim_ctrl"),Weight=1))
    Weight=1
    WorldUp=(Target=(X=0,Y=0,Z=1),Kind=Location,Space=(Type=Bone,Name="pelvis"))
   )
   > &RigUnit_AimConstraintLocalSpaceOffset(
    Child=(Type=Bone,Name="master_back_chain_1_jnt")
    Filter=(bX=True,bY=True,bZ=True)
    Parents=((Item=(Type=Control,Name="thigh_jacket_rear_aim_ctrl"),Weight=1))
    Weight=1
    WorldUp=(Target=(X=0,Y=0,Z=1),Kind=Location,Space=(Type=Bone,Name="pelvis"))
   )
   > &RigUnit_AimConstraintLocalSpaceOffset(
    Child=(Type=Bone,Name="master_R_front_chain_1_jnt")
    Filter=(bX=True,bY=True,bZ=True)
    Parents=((Item=(Type=Control,Name="thigh_r_jacket_front_aim_ctrl"),Weight=1))
    Weight=1
    WorldUp=(Target=(X=0,Y=0,Z=1),Kind=Location,Space=(Type=Bone,Name="pelvis"))
   )
   > &RigUnit_AimConstraintLocalSpaceOffset(
    Child=(Type=Bone,Name="master_R_side_front_chain_1_jnt")
    Filter=(bX=True,bY=True,bZ=True)
    Parents=((Item=(Type=Control,Name="thigh_r_jacket_side_aim_ctrl"),Weight=1))
    Weight=1
    WorldUp=(Target=(X=0,Y=0,Z=1),Kind=Location,Space=(Type=Bone,Name="pelvis"))
   )
   > &RigUnit_AimConstraintLocalSpaceOffset(
    Child=(Type=Bone,Name="master_R_side_rear_chain_1_jnt")
    Filter=(bX=True,bY=True,bZ=True)
    Parents=((Item=(Type=Control,Name="thigh_r_jacket_side_rear_aim_ctrl"),Weight=1))
    Weight=1
    WorldUp=(Target=(X=0,Y=0,Z=1),Kind=Location,Space=(Type=Bone,Name="pelvis"))
   )
   RigUnit_GetTransform = &RigUnit_GetTransform(Item=(Type=Bone,Name="hand_r"), bInitial=False)
   RigVMFunction_MathVectorDistance = &RigVMFunction_MathVectorDistance(
    A=@RigUnit_GetTransform.Transform.Translation
    B=&RigUnit_GetTransform(Item=(Type=Bone,Name="dyn_collar_MSTR_R"), bInitial=False).Transform.Translation
   )
   RigVMFunction_AnimEvalRichCurve = &RigVMFunction_AnimEvalRichCurve(
    Curve=(EditorCurveData=(Keys=((),(Time=0.2516,Value=0.5992),(Time=0.4962,Value=1),(Time=0.7474,Value=0.6055),(Time=1))))
    SourceMaximum=16
    SourceMinimum=11
    TargetMinimum=0
    Value=@RigVMFunction_MathVectorDistance.Result
   )
   RigVMFunction_AnimEasing = /Script/RigVM.RigVMFunction_AnimEasing(SourceMaximum=1, SourceMinimum=0, TargetMaximum=1, TargetMinimum=0, Type=CubicEaseInOut, Value=@RigVMFunction_AnimEvalRichCurve.Result) {is=struct RigVMUnitNode}
   &RigUnit_ModifyTransforms(
    ItemToModify=((Item=(Type=Bone,Name="dyn_collarB_MSTR_R"),Transform=(Rotation=(X=0,Y=0,Z=0,W=1),Translation=(X=0,Y=-5,Z=0),Scale3D=(X=1,Y=1,Z=1))))
    WeightMinimum=0
    Weight=@RigVMFunction_AnimEasing.Result
    ExecutePin=@SequenceAggregate_3.D
   )
  ]}
 RigVMModel Rig {nodes [
   n1 = &RigUnit_PrepareForExecution()
   SequenceAggregate = RigVMAggregateNode(ExecuteContext=@n1.ExecutePin)
   RigUnit_ItemArray = &RigUnit_ItemArray(Items=((Type=Control,Name="thigh_l_jacket_front_aim_ctrl"),(Type=Control,Name="thigh_l_jacket_side_aim_ctrl")))
   RigVMVariableNode(Value=@RigUnit_ItemArray.Items, ExecuteContext=@SequenceAggregate.A) {variable=l_thigh_aim_at_ctrls}
   > RigVMVariableNode(Value=&RigUnit_HierarchyGetPoseItemArray(ElementType=All, Initial=False, ItemsToGet=RigVMVariableNode() {variable=l_thigh_aim_at_ctrls}.Value).Pose) {variable=l_thigh_aim_at_ctrls_bind_pose}
   > RigVMVariableNode(Value=&RigUnit_ItemArray(Items=((Type=Control,Name="thigh_r_jacket_front_aim_ctrl"),(Type=Control,Name="thigh_r_jacket_side_aim_ctrl"))).Items) {variable=r_thigh_aim_at_ctrls}
   > RigVMVariableNode(Value=&RigUnit_HierarchyGetPoseItemArray(ElementType=All, Initial=False, ItemsToGet=RigVMVariableNode() {variable=r_thigh_aim_at_ctrls}.Value).Pose) {variable=r_thigh_aim_at_ctrls_bind_pose}
   &RigUnit_ModifyTransforms(
    ItemToModify="((Item=(Type=Control,Name=\"thigh_l_jacket_front_aim_ctrl\"),Transform=(Rotation=(X=0,Y=0,Z=0,W=1),Translation=(X=5,Y=-5,Z=-13),Scale3D=(X=1,Y=1,Z=1))),(Item=(Type=Control,Name=\"thigh_l_jacket_side_aim_ctrl\"),Transform=(Rotation=(X=0,Y=0,Z=0,W=1),Translation ...(truncated)"
    Weight=1
    WeightMinimum=0
    ExecutePin=@SequenceAggregate.B
   )
   > RigVMVariableNode(Value=&RigUnit_HierarchyGetPoseItemArray(ElementType=All, Initial=False, ItemsToGet=RigVMVariableNode() {variable=l_thigh_aim_at_ctrls}.Value).Pose) {variable=l_thigh_aim_at_ctrls_thigh_up_pose}
   > &RigUnit_ModifyTransforms(
    ItemToModify="((Item=(Type=Control,Name=\"thigh_l_jacket_front_aim_ctrl\"),Transform=(Rotation=(X=0,Y=0,Z=0,W=1),Translation=(X=0,Y=0,Z=0),Scale3D=(X=1,Y=1,Z=1))),(Item=(Type=Control,Name=\"thigh_l_jacket_side_aim_ctrl\"),Transform=(Rotation=(X=0,Y=0,Z=0,W=1),Translation=(X ...(truncated)"
    Weight=1
    WeightMinimum=0
   )
   > RigVMVariableNode(Value=&RigUnit_HierarchyGetPoseItemArray(ElementType=All, Initial=False, ItemsToGet=RigVMVariableNode() {variable=l_thigh_aim_at_ctrls}.Value).Pose) {variable=l_thigh_aim_at_ctrls_thigh_side_pose}
   > &RigUnit_ModifyTransforms(
    ItemToModify="((Item=(Type=Control,Name=\"thigh_r_jacket_front_aim_ctrl\"),Transform=(Rotation=(X=0,Y=0,Z=0,W=1),Translation=(X=5,Y=-5,Z=13),Scale3D=(X=1,Y=1,Z=1))),(Item=(Type=Control,Name=\"thigh_r_jacket_side_aim_ctrl\"),Transform=(Rotation=(X=0,Y=0,Z=0,W=1),Translation= ...(truncated)"
    Weight=1
    WeightMinimum=0
   )
   > RigVMVariableNode(Value=&RigUnit_HierarchyGetPoseItemArray(ElementType=All, Initial=False, ItemsToGet=RigVMVariableNode() {variable=r_thigh_aim_at_ctrls}.Value).Pose) {variable=r_thigh_aim_at_ctrls_thigh_up_pose}
   > &RigUnit_ModifyTransforms(
    ItemToModify="((Item=(Type=Control,Name=\"thigh_r_jacket_front_aim_ctrl\"),Transform=(Rotation=(X=0,Y=0,Z=0,W=1),Translation=(X=0,Y=0,Z=0),Scale3D=(X=1,Y=1,Z=1))),(Item=(Type=Control,Name=\"thigh_r_jacket_side_aim_ctrl\"),Transform=(Rotation=(X=0,Y=0,Z=0,W=1),Translation=(X ...(truncated)"
    Weight=1
    WeightMinimum=0
   )
  ]}
 SequenceAggregate_1_ContainedGraph {nodes [
   n1 = RigVMFunctionEntryNode()
   SequenceAggregate = &RigVMFunction_Sequence(ExecuteContext=@n1.ExecuteContext)
   SequenceAggregate_1 = &RigVMFunction_Sequence(ExecuteContext=@SequenceAggregate.B)
   RigVMFunctionReturnNode(A=@SequenceAggregate.A, B=@SequenceAggregate_1.A, C=@SequenceAggregate_1.B)
  ]}
 SequenceAggregate_3_1_ContainedGraph {nodes [
   n1 = RigVMFunctionEntryNode()
   SequenceAggregate_3 = &RigVMFunction_Sequence(ExecuteContext=@n1.ExecuteContext)
   SequenceAggregate_3_1 = &RigVMFunction_Sequence(ExecuteContext=@SequenceAggregate_3.B)
   SequenceAggregate_3_2 = &RigVMFunction_Sequence(ExecuteContext=@SequenceAggregate_3_1.B)
   RigVMFunctionReturnNode(A=@SequenceAggregate_3.A, B=@SequenceAggregate_3_1.A, C=@SequenceAggregate_3_2.A, D=@SequenceAggregate_3_2.B)
  ]}
]
hierarchy [
 root {loc=<0 0 0>, rot=<0 0 0 1>}
 pelvis {parent=root, loc=<0 -0.5301 98.0571>, rot=<0.0127 0.707 -0.0127 -0.707>}
 spine_01 {parent=pelvis, loc=<6.1208 -0.9328 0>, rot=<0 0 0.034 -0.9994>}
 spine_02 {parent=spine_01, loc=<6.5996 -0.2095 0>, rot=<0 0 0.0238 -0.9997>}
 spine_03 {parent=spine_02, loc=<6.4914 -0.1237 0>, rot=<0 0 0.0149 -0.9999>}
 spine_04 {parent=spine_03, loc=<6.9598 0.0794 0>, rot=<0 0 -0.0132 -0.9999>}
 spine_05 {parent=spine_04, loc=<6.9987 0 0>, rot=<0 0 -0.0015 -1>}
 attach_cape {parent=spine_05, loc=<9.3216 -8.4501 0>, rot=<0.0616 -0.7044 -0.0616 -0.7044>}
 neck_01 {parent=spine_05, loc=<11.3138 0.1529 0>, rot=<0 0 -0.2003 -0.9797>}
 neck_02 {parent=neck_01, loc=<4.5627 0 0>, rot=<0 0 0.0504 -0.9987>}
 head {parent=neck_02, loc=<5.3914 0 0>, rot=<0 0 0.115 -0.9934>}
 dyn_hair_back_up_l {parent=head, loc=<5.811 -5.3596 -5.8742>, rot=<0.0037 -0.8866 0.4625 -0.0019>}
 dyn_hair_back_down_l {parent=head, loc=<-1.3898 -1.9778 -6.0192>, rot=<-0.3508 -0.9025 -0.1674 0.1856>}
 dyn_hair_back_up_r {parent=head, loc=<5.8087 -5.0467 6.3228>, rot=<0.8866 0.0037 -0.0019 -0.4625>}
 dyn_hair_back_down_r {parent=head, loc=<-1.3899 -1.9778 6.0192>, rot=<0.9025 -0.3508 0.1856 0.1674>}
 dyn_hair_back {parent=head, loc=<6.1274 -7.4072 0>, rot=<0.0029 -0.7071 0.7071 -0.0029>}
 dyn_hair_right_strand_A_01 {parent=head, loc=<7.4285 4.6764 6.364>, rot=<0.983 0.0991 -0.0077 0.1545>}
 dyn_hair_right_strand_A_02 {parent=dyn_hair_right_strand_A_01, loc=<-5.5715 0 0>, rot=<0 0 0 1>}
 dyn_hair_left_strand_A_01 {parent=head, loc=<5.6085 2.8385 -6.6931>, rot=<0.0568 -0.991 -0.1214 -0.006>}
 dyn_hair_left_strand_A_02 {parent=dyn_hair_left_strand_A_01, loc=<3.1035 0 0>, rot=<0 0 0 1>}
 dyn_hair_left_strand_B_01 {parent=head, loc=<5.3408 5.6243 -6.0023>, rot=<0.0987 -0.9877 -0.121 -0.0111>}
 dyn_hair_front_strand_A_01 {parent=head, loc=<8.9678 8.5389 -0.9566>, rot=<-0.1217 -0.6032 -0.3718 -0.6951>}
 dyn_hair_front_strand_A_02 {parent=dyn_hair_front_strand_A_01, loc=<2.2361 0 0>, rot=<0 -0.6081 0 -0.7939>}
 dyn_hair_front_strand_A_03 {parent=dyn_hair_front_strand_A_02, loc=<3.3622 0 0>, rot=<0 0.4111 0 -0.9116>}
 dyn_hair_front_strand_B_01 {parent=head, loc=<7.5432 8.0829 4.4539>, rot=<0.8569 0.1758 -0.4744 0.0989>}
 dyn_hair_front_strand_C_01 {parent=head, loc=<8.0055 9.3146 1.1671>, rot=<0.8045 0.0892 -0.3389 0.4796>}
 dyn_zipperA_mstr {parent=neck_01, loc=<1.8105 4.1423 0.0006>, rot=<0 0 -0.5842 -0.8116>}
 dyn_zipperA {parent=dyn_zipperA_mstr, loc=<0 0 0>, rot=<0 0 0 1>}
 attach_backpack {parent=spine_05, loc=<-0.123 -13.473 0>, rot=<-0.0093 -0.707 0.0093 -0.707>}
 clavicle_l {parent=spine_05, loc=<5.4831 -0.1053 -3.2946>, rot=<0.0465 -0.7289 -0.009 -0.6829>}
 upperarm_l {parent=clavicle_l, loc=<9.8658 -1.2136 -0.4105>, rot=<0.0164 -0.3245 -0.0269 -0.9454>}
 lowerarm_l {parent=upperarm_l, loc=<26.465 -0.8345 0.259>, rot=<0.0036 0.0161 -0.2298 -0.9731>}
 hand_l {parent=lowerarm_l, loc=<22.7355 -0.3203 -0.6105>, rot=<-0.6784 -0.0223 0.0735 -0.7307>}
 thumb_01_l {parent=hand_l, loc=<1.3118 -0.3191 -2.459>, rot=<0.6546 -0.4042 0.0664 -0.6354>}
 thumb_02_l {parent=thumb_01_l, loc=<5.9336 -0.0448 0.0401>, rot=<-0.0037 0.0213 0.298 -0.9543>}
 thumb_03_l {parent=thumb_02_l, loc=<2.2223 0.3682 -0.0907>, rot=<0.0118 -0.013 0.0066 -0.9998>}
 index_metacarpal_l {parent=hand_l, loc=<1.6535 0.0183 -2.96>, rot=<0.0158 -0.0124 -0.0173 -0.9996>}
 index_01_l {parent=index_metacarpal_l, loc=<8.3008 0.0357 -0.0524>, rot=<-0.0044 0.0485 0.0924 -0.9945>}
 index_02_l {parent=index_01_l, loc=<3.3587 -0.3004 0.0452>, rot=<-0.002 0.0027 0.3225 -0.9466>}
 index_03_l {parent=index_02_l, loc=<2.5578 -0.123 -0.0252>, rot=<-0.002 0.0067 0.2558 -0.9667>}
 middle_metacarpal_l {parent=hand_l, loc=<2.1856 -0.1725 -0.5984>, rot=<0.006 0.0168 -0.0329 -0.9993>}
 middle_01_l {parent=middle_metacarpal_l, loc=<7.3275 -0.2447 0.0129>, rot=<-0.0208 0.0042 0.0858 -0.9961>}
 middle_02_l {parent=middle_01_l, loc=<4.0123 0.1401 0.1175>, rot=<-0.0218 0.0258 0.3512 -0.9357>}
 middle_03_l {parent=middle_02_l, loc=<2.6077 0.0534 0.1373>, rot=<0.0082 0.0188 0.1708 -0.9851>}
 ring_metacarpal_l {parent=hand_l, loc=<2.2074 -0.3625 1.3638>, rot=<-0.0284 0.0722 -0.0458 -0.9959>}
 ring_01_l {parent=ring_metacarpal_l, loc=<7.0295 -0.6197 -0.1446>, rot=<-0.0489 -0.0252 0.1892 -0.9804>}
 ring_02_l {parent=ring_01_l, loc=<3.6174 0.4647 -0.1027>, rot=<0 0 0.2806 -0.9598>}
 ring_03_l {parent=ring_02_l, loc=<2.4152 0.114 0.0199>, rot=<0 0 0.1189 -0.9929>}
 pinky_metacarpal_l {parent=hand_l, loc=<1.7616 -0.9753 3.3814>, rot=<-0.0397 0.0814 -0.069 -0.9935>}
 pinky_01_l {parent=pinky_metacarpal_l, loc=<7.1709 -0.4831 -0.1094>, rot=<-0.0654 -0.0035 0.2242 -0.9723>}
 pinky_02_l {parent=pinky_01_l, loc=<2.8274 0.0893 -0.1043>, rot=<0.0037 0.0011 0.2854 -0.9584>}
 pinky_03_l {parent=pinky_02_l, loc=<2.2046 0.0307 -0.0078>, rot=<-0.0002 0.0098 0.1445 -0.9895>}
 deform_wrist_dn_base_l {parent=hand_l, loc=<0 -1.5326 0>, rot=<0 0 0 -1>}
 deform_wrist_dn_l {parent=deform_wrist_dn_base_l, loc=<0 0 0>, rot=<0 0 0 1>}
 deform_wrist_up_base_l {parent=hand_l, loc=<0 2.3009 0>, rot=<0 0 0 -1>}
 deform_wrist_up_l {parent=deform_wrist_up_base_l, loc=<0 0 0>, rot=<0 0 0 1>}
 weapon_l {parent=hand_l, loc=<-0.1974 -3.4111 0.3811>, rot=<0 0 0 1>}
 hand_attach_l {parent=hand_l, loc=<8.5 -3 0>, rot=<0.7071 0.7071 0 0>}
 lowerarm_twist_02_l {parent=lowerarm_l, loc=<10.6707 0 0>, rot=<0 0 0 1>}
 lowerarm_twist_01_l {parent=lowerarm_l, loc=<18.316 0 0>, rot=<0 0 0 1>}
 deform_elbow_in_base_l {parent=lowerarm_l, loc=<0.1834 3.6062 -0.2572>, rot=<0 0 0 -1>}
 deform_elbow_in_l {parent=deform_elbow_in_base_l, loc=<0 0 0>, rot=<0 0 0 1>}
 deform_elbow_out_base_l {parent=lowerarm_l, loc=<0 -4 0>, rot=<0 0 0 1>}
 deform_elbow_out_l {parent=deform_elbow_out_base_l, loc=<0 0 0>, rot=<0 0 0 1>}
 upperarm_twist_01_l {parent=upperarm_l, loc=<7.2135 -0.1134 0>, rot=<0 0 0 1>}
 deform_upperarm_bk_base_l {parent=upperarm_twist_01_l, loc=<-6.4053 -4.3746 0.314>, rot=<0 0 0 -1>}
 deform_upperarm_bk_l {parent=deform_upperarm_bk_base_l, loc=<0 0 0>, rot=<0 0 0 1>}
 deform_upperarm_fr_base_l {parent=upperarm_twist_01_l, loc=<-5.6312 5.092 0.3354>, rot=<0 0 0 -1>}
 deform_upperarm_fr_l {parent=deform_upperarm_fr_base_l, loc=<0 0 0>, rot=<0 0 0 1>}
 deform_upperarm_up_base_l {parent=upperarm_twist_01_l, loc=<-8.891 0.4825 3.9517>, rot=<0 0 0 -1>}
 deform_upperarm_up_l {parent=deform_upperarm_up_base_l, loc=<0 0 0>, rot=<0 0 0 1>}
 upperarm_twist_02_l {parent=upperarm_l, loc=<19.7998 -0.1134 0>, rot=<0 0 0 1>}
 clavicle_r {parent=spine_05, loc=<5.4831 -0.1053 3.2946>, rot=<0.7289 0.0465 -0.6829 0.009>}
 upperarm_r {parent=clavicle_r, loc=<-9.8658 1.2136 0.4105>, rot=<0.0164 -0.3245 -0.0269 -0.9454>}
 lowerarm_r {parent=upperarm_r, loc=<-26.465 0.8345 -0.259>, rot=<0.0036 0.0161 -0.2298 -0.9731>}
 hand_r {parent=lowerarm_r, loc=<-22.7355 0.3203 0.6106>, rot=<-0.6784 -0.0223 0.0735 -0.7307>}
 thumb_01_r {parent=hand_r, loc=<-1.3118 0.3191 2.459>, rot=<0.6546 -0.4042 0.0664 -0.6354>}
 thumb_02_r {parent=thumb_01_r, loc=<-5.9336 0.0448 -0.0401>, rot=<-0.0037 0.0213 0.298 -0.9543>}
 thumb_03_r {parent=thumb_02_r, loc=<-2.2223 -0.3682 0.0907>, rot=<0.0118 -0.013 0.0066 -0.9998>}
 index_metacarpal_r {parent=hand_r, loc=<-1.6535 -0.0183 2.96>, rot=<0.0158 -0.0124 -0.0173 -0.9996>}
 index_01_r {parent=index_metacarpal_r, loc=<-8.3008 -0.0357 0.0524>, rot=<-0.0044 0.0485 0.0924 -0.9945>}
 index_02_r {parent=index_01_r, loc=<-3.3587 0.3004 -0.0452>, rot=<-0.002 0.0027 0.3225 -0.9466>}
 index_03_r {parent=index_02_r, loc=<-2.5578 0.123 0.0252>, rot=<-0.002 0.0067 0.2558 -0.9667>}
 middle_metacarpal_r {parent=hand_r, loc=<-2.1856 0.1725 0.5984>, rot=<0.006 0.0168 -0.0329 -0.9993>}
 middle_01_r {parent=middle_metacarpal_r, loc=<-7.3275 0.2447 -0.0129>, rot=<-0.0208 0.0042 0.0858 -0.9961>}
 middle_02_r {parent=middle_01_r, loc=<-4.0123 -0.1401 -0.1175>, rot=<-0.0218 0.0258 0.3512 -0.9357>}
 middle_03_r {parent=middle_02_r, loc=<-2.6077 -0.0534 -0.1373>, rot=<0.0082 0.0188 0.1708 -0.9851>}
 ring_metacarpal_r {parent=hand_r, loc=<-2.2074 0.3625 -1.3638>, rot=<-0.0284 0.0722 -0.0458 -0.9959>}
 ring_01_r {parent=ring_metacarpal_r, loc=<-7.0295 0.6197 0.1446>, rot=<-0.0489 -0.0252 0.1892 -0.9804>}
 ring_02_r {parent=ring_01_r, loc=<-3.6174 -0.4647 0.1027>, rot=<0 0 0.2806 -0.9598>}
 ring_03_r {parent=ring_02_r, loc=<-2.4152 -0.114 -0.0199>, rot=<0 0 0.1189 -0.9929>}
 pinky_metacarpal_r {parent=hand_r, loc=<-1.7616 0.9753 -3.3814>, rot=<-0.0397 0.0814 -0.069 -0.9935>}
 pinky_01_r {parent=pinky_metacarpal_r, loc=<-7.1709 0.4831 0.1094>, rot=<-0.0654 -0.0035 0.2242 -0.9723>}
 pinky_02_r {parent=pinky_01_r, loc=<-2.8274 -0.0893 0.1043>, rot=<0.0037 0.0011 0.2854 -0.9584>}
 pinky_03_r {parent=pinky_02_r, loc=<-2.2046 -0.0307 0.0078>, rot=<-0.0002 0.0098 0.1445 -0.9895>}
 deform_wrist_dn_base_r {parent=hand_r, loc=<-0.0001 1.5328 0>, rot=<0 0 0 -1>}
 deform_wrist_dn_r {parent=deform_wrist_dn_base_r, loc=<0 0 0>, rot=<0 0 0 1>}
 deform_wrist_up_base_r {parent=hand_r, loc=<-0.0002 -2.3007 -0.0001>, rot=<0 0 0 -1>}
 deform_wrist_up_r {parent=deform_wrist_up_base_r, loc=<0 0 0>, rot=<0 0 0 1>}
 weapon_r {parent=hand_r, loc=<0.1974 3.4111 -0.3811>, rot=<0 0 0 1>}
 hand_attach_r {parent=hand_r, loc=<-8.5001 2.9999 0>, rot=<0 0 -0.7071 -0.7071>}
 lowerarm_twist_02_r {parent=lowerarm_r, loc=<-10.6707 0 0>, rot=<0 0 0 -1>}
 lowerarm_twist_01_r {parent=lowerarm_r, loc=<-18.316 0 0>, rot=<0 0 0 -1>}
 deform_elbow_in_base_r {parent=lowerarm_r, loc=<-0.1834 -3.6062 0.2572>, rot=<0 0 0 -1>}
 deform_elbow_in_r {parent=deform_elbow_in_base_r, loc=<0 0 0>, rot=<0 0 0 1>}
 deform_elbow_out_base_r {parent=lowerarm_r, loc=<0 4 -0.0001>, rot=<0 0 0 -1>}
 deform_elbow_out_r {parent=deform_elbow_out_base_r, loc=<0 0 0>, rot=<0 0 0 1>}
 upperarm_twist_01_r {parent=upperarm_r, loc=<-7.2135 0.1134 0>, rot=<0 0 0 -1>}
 deform_upperarm_bk_base_r {parent=upperarm_twist_01_r, loc=<6.4053 4.3746 -0.314>, rot=<0 0 0 -1>}
 deform_upperarm_bk_r {parent=deform_upperarm_bk_base_r, loc=<0 0 0>, rot=<0 0 0 1>}
 deform_upperarm_fr_base_r {parent=upperarm_twist_01_r, loc=<5.6312 -5.092 -0.3354>, rot=<0 0 0 -1>}
 deform_upperarm_fr_r {parent=deform_upperarm_fr_base_r, loc=<0 0 0>, rot=<0 0 0 1>}
 deform_upperarm_up_base_r {parent=upperarm_twist_01_r, loc=<8.891 -0.4825 -3.9517>, rot=<0 0 0 -1>}
 deform_upperarm_up_r {parent=deform_upperarm_up_base_r, loc=<0 0 0>, rot=<0 0 0 1>}
 upperarm_twist_02_r {parent=upperarm_r, loc=<-19.7998 0.1134 0>, rot=<0 0 0 -1>}
 deform_upperarm_twist_ADS_correction {parent=upperarm_twist_02_r, loc=<6.6067 -2.6711 4.3332>, rot=<-0.2612 0.6223 0.2321 -0.7005>}
 dyn_collar_bk_MSTR_R {parent=spine_05, loc=<11.796 -1.7352 5.1508>, rot=<0.0452 -0.9355 -0.3218 -0.1388>}
 dyn_collar_bk_01_R {parent=dyn_collar_bk_MSTR_R, loc=<-1.6889 0 -0.0001>, rot=<0 0 0 1>}
 dyn_collar_bk_MSTR {parent=spine_05, loc=<12.3359 -4.4325 0.0895>, rot=<-0.697 -0.0411 0.0034 -0.7158>}
 dyn_collar_bk_01 {parent=dyn_collar_bk_MSTR, loc=<1.5396 0 0>, rot=<0 0 0 1>}
 dyn_collar_bk_MSTR_L {parent=spine_05, loc=<11.7956 -1.7352 -5.1508>, rot=<0.9355 0.0452 -0.1388 0.3218>}
 dyn_collar_bk_01_L {parent=dyn_collar_bk_MSTR_L, loc=<1.6891 0 0>, rot=<0 0 0 1>}
 dyn_collar_MSTR_L {parent=spine_05, loc=<8.5137 3.1803 -7.2038>, rot=<0.9393 0.2531 -0.1438 0.1816>}
 dyn_collar_01_L {parent=dyn_collar_MSTR_L, loc=<1.8635 0 0>, rot=<0 0 0 1>}
 dyn_collar_MSTR_R {parent=spine_05, loc=<8.5135 3.1802 7.2038>, rot=<0.2531 -0.9393 -0.1816 -0.1438>}
 dyn_collar_01_R {parent=dyn_collar_MSTR_R, loc=<-1.8636 -0.0001 0>, rot=<0 0 0 1>}
 dyn_collarB_MSTR_R {parent=spine_05, loc=<0.368 10.7218 8.3421>, rot=<0.292 -0.9078 -0.2761 -0.1201>}
 dyn_collarB_01_R {parent=dyn_collarB_MSTR_R, loc=<-1.8632 0.0002 0>, rot=<0 0 0 1>}
 dyn_collarB_MSTR_L {parent=spine_05, loc=<0.3679 10.7218 -8.3421>, rot=<0.9078 0.292 -0.1201 0.2761>}
 dyn_collarB_01_L {parent=dyn_collarB_MSTR_L, loc=<1.8635 0 0>, rot=<0 0 0 1>}
 deform_jacket_armpit_r_corrector {parent=spine_04, loc=<0 0 13>, rot=<-0.5 0.5 -0.5 -0.5>}
 deform_jacket_armpit_l_corrector {parent=spine_04, loc=<0 0 -13>, rot=<0.5 -0.5 -0.5 -0.5>}
 master_jacket_jnt {parent=spine_03, loc=<0 0 0>, rot=<0 0 0 1>}
 master_L_front_chain_1_jnt {parent=master_jacket_jnt, loc=<-1.0607 8.5228 -11.0668>, rot=<0.0213 -0.9456 -0.3238 -0.0229>}
 dyn_L_front_chain_1_jnt {parent=master_L_front_chain_1_jnt, loc=<0 0 0>, rot=<0 0 0 1>}
 dyn_L_front_chain_2_jnt1 {parent=dyn_L_front_chain_1_jnt, loc=<8.1525 0 0>, rot=<0 0 0 1>}
 dyn_L_front_chain_3_jnt {parent=dyn_L_front_chain_2_jnt1, loc=<8.068 0 0>, rot=<0 0 0 1>}
 master_L_side_front_chain_1_jnt {parent=master_jacket_jnt, loc=<0.0343 -0.3116 -14.2407>, rot=<0 -0.9964 0 -0.0843>}
 dyn_L_side_front_chain_1_jnt {parent=master_L_side_front_chain_1_jnt, loc=<0 0 0>, rot=<0 0 0 1>}
 dyn_L_side_front_chain_2_jnt {parent=dyn_L_side_front_chain_1_jnt, loc=<8.3455 0 0>, rot=<0.0015 -0.0002 0.0105 -0.9999>}
 dyn_L_side_front_chain_3_jnt {parent=dyn_L_side_front_chain_2_jnt, loc=<8.205 0 0>, rot=<0 0 0 1>}
 master_L_side_rear_chain_1_jnt {parent=master_jacket_jnt, loc=<1.047 -9.345 -11.2952>, rot=<0.0226 -0.9406 0.3341 -0.0552>}
 dyn_L_side_rear_chain_1_jnt {parent=master_L_side_rear_chain_1_jnt, loc=<0 0 0>, rot=<0 0 0 1>}
 dyn_L_side_rear_chain_2_jnt {parent=dyn_L_side_rear_chain_1_jnt, loc=<8.2922 0 0>, rot=<0 -0.016 -0.0002 -0.9999>}
 dyn_L_side_rear_chain_3_jnt {parent=dyn_L_side_rear_chain_2_jnt, loc=<8.0355 0 0>, rot=<0 0 0 1>}
 master_back_chain_1_jnt {parent=master_jacket_jnt, loc=<1.2006 -12.0415 0.0009>, rot=<0.0115 -0.705 0.709 -0.0148>}
 dyn_back_chain_1_jnt {parent=master_back_chain_1_jnt, loc=<0 0 0>, rot=<0 0 0 1>}
 dyn_back_chain_2_jnt {parent=dyn_back_chain_1_jnt, loc=<8.4823 -0.0352 0.7687>, rot=<0 0 0 1>}
 dyn_back_chain_3_jnt {parent=dyn_back_chain_2_jnt, loc=<7.5012 -0.0311 0.6798>, rot=<0 0 0 1>}
 master_R_side_rear_chain_1_jnt {parent=master_jacket_jnt, loc=<1.047 -9.3449 11.2952>, rot=<0.9406 0.0226 -0.0552 -0.3341>}
 dyn_R_side_rear_chain_1_jnt {parent=master_R_side_rear_chain_1_jnt, loc=<0 0 0>, rot=<0 0 0 1>}
 dyn_R_side_rear_chain_2_jnt {parent=dyn_R_side_rear_chain_1_jnt, loc=<-8.2923 0 0>, rot=<0 -0.016 -0.0002 -0.9999>}
 dyn_R_side_rear_chain_3_jnt {parent=dyn_R_side_rear_chain_2_jnt, loc=<-8.0357 0 0>, rot=<0 0 0 1>}
 master_R_side_front_chain_1_jnt {parent=master_jacket_jnt, loc=<0.0345 -0.3116 14.2407>, rot=<0.9964 0 -0.0843 0>}
 dyn_R_side_front_chain_1_jnt {parent=master_R_side_front_chain_1_jnt, loc=<0 0 0>, rot=<0 0 0 1>}
 dyn_R_side_front_chain_2_jnt {parent=dyn_R_side_front_chain_1_jnt, loc=<-8.3458 0 0>, rot=<0.0015 -0.0002 0.0105 -0.9999>}
 dyn_R_side_front_chain_3_jnt {parent=dyn_R_side_front_chain_2_jnt, loc=<-8.2054 0 0.0001>, rot=<0 0 0 1>}
 master_R_front_chain_1_jnt {parent=master_jacket_jnt, loc=<-1.0607 8.5228 11.0668>, rot=<0.9456 0.0213 -0.0229 0.3238>}
 dyn_R_front_chain_1_jnt {parent=master_R_front_chain_1_jnt, loc=<0 0 0>, rot=<0 0 0 1>}
 dyn_R_front_chain_2_jnt1 {parent=dyn_R_front_chain_1_jnt, loc=<-8.1529 0 0.0001>, rot=<0 0 0 1>}
 dyn_R_front_chain_3_jnt {parent=dyn_R_front_chain_2_jnt1, loc=<-8.0675 0 0>, rot=<0 0 0 1>}
 thigh_l {parent=pelvis, loc=<-1.0855 -0.2109 -8.7053>, rot=<0.0009 0.0355 0.0155 -0.9993>}
 calf_l {parent=thigh_l, loc=<-42.5284 -0.0019 -0.7297>, rot=<0.0006 -0.0161 -0.0391 -0.9991>}
 foot_l {parent=calf_l, loc=<-46.7401 -2.5014 -0.592>, rot=<0.0427 -0.0238 0.0353 -0.9982>}
 ball_l {parent=foot_l, loc=<-3.5658 10.8066 -0.1544>, rot=<0.0001 -0.0001 -0.7063 -0.7079>}
 calf_twist_02_l {parent=calf_l, loc=<-19.1546 -1.7943 -0.2638>, rot=<0.0237 0.0095 -0.0035 -0.9997>}
 calf_twist_01_l {parent=calf_l, loc=<-35.8483 -1.9197 -0.5794>, rot=<0.0237 0.0095 -0.0035 -0.9997>}
 deform_knee_in_base_l {parent=calf_l, loc=<0.8031 -5.0231 0.2456>, rot=<0 0 0 -1>}
 deform_knee_in_l {parent=deform_knee_in_base_l, loc=<0 0 0>, rot=<0 0 0 1>}
 deform_knee_out_base_l {parent=calf_l, loc=<0 4 0>, rot=<0 0 0 -1>}
 deform_knee_out_l {parent=deform_knee_out_base_l, loc=<0 0 0>, rot=<0 0 0 1>}
 thigh_twist_01_l {parent=thigh_l, loc=<-27.952 0 0>, rot=<0.0013 0 0 -1>}
 dyn_zipperB_mstr {parent=thigh_l, loc=<-12.2678 6.4817 -7.0029>, rot=<0.3012 -0.3474 -0.6379 -0.6179>}
 dyn_zipperB {parent=dyn_zipperB_mstr, loc=<0 0 0>, rot=<0 0 0 1>}
 deform_groin_base_l {parent=thigh_l, loc=<0.4801 5.6776 -0.2486>, rot=<0 0 0 -1>}
 deform_groin_l {parent=deform_groin_base_l, loc=<0 0 0>, rot=<0 0 0 1>}
 thigh_r {parent=pelvis, loc=<-1.0855 -0.2109 8.7053>, rot=<0.0355 -0.0009 0.9993 0.0155>}
 calf_r {parent=thigh_r, loc=<42.5284 0.0019 0.7297>, rot=<0.0006 -0.0161 -0.0391 -0.9991>}
 foot_r {parent=calf_r, loc=<46.7401 2.5014 0.592>, rot=<0.0427 -0.0238 0.0353 -0.9982>}
 ball_r {parent=foot_r, loc=<3.5658 -10.8066 0.1544>, rot=<0.0001 -0.0001 -0.7063 -0.7079>}
 calf_twist_02_r {parent=calf_r, loc=<19.1546 1.7943 0.2638>, rot=<0.0237 0.0095 -0.0035 -0.9997>}
 calf_twist_01_r {parent=calf_r, loc=<35.8483 1.9197 0.5794>, rot=<0.0237 0.0095 -0.0035 -0.9997>}
 deform_knee_in_base_r {parent=calf_r, loc=<-0.8031 5.0231 -0.2457>, rot=<0 0 0 -1>}
 deform_knee_in_r {parent=deform_knee_in_base_r, loc=<0 0 0>, rot=<0 0 0 1>}
 deform_knee_out_base_r {parent=calf_r, loc=<0 -4 0>, rot=<0 0 0 -1>}
 deform_knee_out_r {parent=deform_knee_out_base_r, loc=<0 0 0>, rot=<0 0 0 1>}
 thigh_twist_01_r {parent=thigh_r, loc=<27.952 0 0>, rot=<0.0013 0 0 -1>}
 dyn_zipperC_mstr {parent=thigh_r, loc=<12.2634 -6.4841 7.0017>, rot=<0.343 0.2968 0.6203 -0.6399>}
 dyn_zipperC {parent=dyn_zipperC_mstr, loc=<0 0 0>, rot=<0 0 0 1>}
 deform_groin_base_r {parent=thigh_r, loc=<-0.4801 -5.6776 0.2486>, rot=<0 0 0 -1>}
 deform_groin_r {parent=deform_groin_base_r, loc=<0 0 0>, rot=<0 0 0 1>}
 deform_glute_base_l {parent=pelvis, loc=<-7.3466 -9.0213 -9.1245>, rot=<0.0172 0.031 -0.4477 -0.8935>}
 deform_glute_l {parent=deform_glute_base_l, loc=<0 0 0>, rot=<0 0 0 1>}
 deform_glute_base_r {parent=pelvis, loc=<0 0 0>, rot=<0.031 -0.0172 0.8935 -0.4477>}
 deform_glute_r {parent=deform_glute_base_r, loc=<11.98 -1.0027 8.6071>, rot=<0 0 0 1>}
 thigh_l_jacket_side_aim_ctrl {controlType=EULER_TRANSFORM, parent=pelvis, type=Control, loc=<0 0 0>, rot=<0 0 0 1>}
 thigh_l_jacket_front_aim_ctrl {controlType=EULER_TRANSFORM, parent=pelvis, type=Control, loc=<0 0 0>, rot=<0 0 0 1>}
 thigh_r_jacket_front_aim_ctrl {controlType=EULER_TRANSFORM, parent=pelvis, type=Control, loc=<0 0 0>, rot=<0 0 0 1>}
 thigh_r_jacket_side_aim_ctrl {controlType=EULER_TRANSFORM, parent=pelvis, type=Control, loc=<0 0 0>, rot=<0 0 0 1>}
 thigh_l_jacket_side_rear_aim_ctrl {controlType=EULER_TRANSFORM, parent=pelvis, type=Control, loc=<0 0 0>, rot=<0 0 0 1>}
 thigh_jacket_rear_aim_ctrl {controlType=EULER_TRANSFORM, parent=pelvis, type=Control, loc=<0 0 0>, rot=<0 0 0 1>}
 thigh_r_jacket_side_rear_aim_ctrl {controlType=EULER_TRANSFORM, parent=pelvis, type=Control, loc=<0 0 0>, rot=<0 0 0 1>}
 ik_hand_root {parent=root, loc=<0 0 0>, rot=<0 0 0 1>}
 ik_hand_gun {parent=ik_hand_root, loc=<-48.1743 3.3545 104.4381>, rot=<0.6276 0.1309 0.4183 -0.6434>}
 ik_hand_l {parent=ik_hand_gun, loc=<59.3271 67.6997 34.3524>, rot=<-0.6158 -0.7027 -0.3565 0>}
]
variables [
 RigVMFunction_MathVectorDistance_Result {type=float, default=0}
 deformRig_l_thigh_up_alpha {type=float, default=0}
 deformRig_r_thigh_up_alpha {type=float, default=0}
 deform_bool {
  type=FUserDefinedStruct_#14
  default="(upperarm_1_#4=True,elbow_3_#5=True,wrist_7_#6=True,groin_8_#7=True,glute_17_#8=True,knee_9_40C408324CA261 ...(truncated)"
 }
 deform_value {
  type=FUserDefinedStruct_#15
  default="(upperarm_up_18_#10=(Rotation=(X=0,Y=0,Z=0,W=1),Translation=(X=0,Y=0,Z=8),Scale3D=(X=0.7,Y=1,Z=1)),upperarm_fwd_fr_19_#11=(Rotation=(X=0,Y=0,Z=0,W=1),Translation=(X=0,Y=10,Z=0),Scale3D=(X=0.5,Y=1,Z= ...(truncated)"
 }
 l_thigh_aim_at_ctrls {
  type="TArray<FRigElementKey>"
  default=((Type=Control,Name="thigh_l_jacket_front_aim_ctrl"),(Type=Control,Name="thigh_l_jacket_side_aim_ctrl"))
 }
 l_thigh_aim_at_ctrls_bind_pose {type=FRigPose, default=(Elements=,HierarchyTopologyVersion=-1,PoseHash=-1)}
 l_thigh_aim_at_ctrls_thigh_side_pose {type=FRigPose, default=(Elements=,HierarchyTopologyVersion=-1,PoseHash=-1)}
 l_thigh_aim_at_ctrls_thigh_up_pose {type=FRigPose, default=(Elements=,HierarchyTopologyVersion=-1,PoseHash=-1)}
 r_thigh_aim_at_ctrls {
  type="TArray<FRigElementKey>"
  default=((Type=Control,Name="thigh_l_jacket_front_aim_ctrl"),(Type=Control,Name="thigh_l_jacket_side_aim_ctrl"))
 }
 r_thigh_aim_at_ctrls_bind_pose {type=FRigPose, default=(Elements=,HierarchyTopologyVersion=-1,PoseHash=-1)}
 r_thigh_aim_at_ctrls_thigh_side_pose {type=FRigPose, default=(Elements=,HierarchyTopologyVersion=-1,PoseHash=-1)}
 r_thigh_aim_at_ctrls_thigh_up_pose {type=FRigPose, default=(Elements=,HierarchyTopologyVersion=-1,PoseHash=-1)}
 spr_l_thigh_side_alpha {type=float, default=0}
 spr_r_thigh_side_alpha {type=float, default=0}
]
