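# Physics asset for the jacket hems of the female medium PizzaParty cosmetic body, defining capsule chains for secondary motion of the hem.
# Seven chains (back, plus left and right front, side-front and side-rear) hang from master anchor bodies constrained to spine_02 with limited swing and twist.
# Cross constraints tie neighboring chains at the second segment, and front to side-front at the third, forming a connected skirt.
# Pelvis, lower spine and thigh shapes provide body volumes.
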PhysicsAsset /BRCosmetics/Characters/Player/Female/Medium/Bodies/F_MED_PizzaParty/Meshes/F_MED_PizzaParty_JacketHems_Physics
bodies {
 dyn_L_front_chain_1_jnt [{shapeType=Capsule, length=3.5, radius=3, shapeName=dyn_L_front_chain_1_jnt_capsule, center=<4 -0.7741 1.6861>, rotation=<rot 0 0 95>}]
 dyn_L_front_chain_2_jnt1 [{
   shapeType=Capsule
   length=6.3785
   radius=3.3062
   shapeName=dyn_L_front_chain_2_jnt1_capsule
   center=<3.5625 -1.5834 1.3462>
   rotation=<rot 0.0001 0.0001 95>
  }]
 dyn_L_front_chain_3_jnt [{
   shapeType=Capsule
   length=10.4886
   radius=2.641
   shapeName=dyn_L_front_chain_3_jnt_capsule
   center=<5.4694 -3.6858 0.4785>
   rotation=<rot 0 0.0001 95>
  }]
 dyn_L_side_front_chain_2_jnt [{
   shapeType=Capsule
   length=6.6799
   radius=3.3201
   shapeName=dyn_L_side_front_chain_2_jnt_capsule
   center=<3.7803 -1.124 0.8093>
   rotation=<rot 0 0 90>
  }]
 dyn_L_side_front_chain_3_jnt [{
   shapeType=Capsule
   length=7.6799
   radius=3.3201
   shapeName=dyn_L_side_front_chain_3_jnt_capsule
   center=<4.9849 -0.1054 0.4666>
   rotation=<rot 0 -4.9999 90>
  }]
 dyn_L_side_rear_chain_1_jnt [{shapeType=Capsule, length=6, radius=4, shapeName=dyn_L_side_rear_chain_1_jnt_capsule, center=<3.7803 0 3.2465>, rotation=<rot 0 0 90>}]
 dyn_L_side_rear_chain_2_jnt [{shapeType=Capsule, length=7, radius=3, shapeName=dyn_L_side_rear_chain_2_jnt_capsule, center=<3.7803 0 4.1019>, rotation=<rot 0 0 90>}]
 dyn_L_side_rear_chain_3_jnt [{
   shapeType=Capsule
   length=6
   radius=4
   shapeName=dyn_L_side_rear_chain_3_jnt_capsule
   center=<5.1986 0.1214 0.0271>
   rotation=<rot -0.0003 0.0003 80>
  }]
 dyn_R_front_chain_1_jnt [{shapeType=Capsule, length=3.5, radius=3, shapeName=dyn_L_front_chain_1_jnt_capsule, center=<-4 0.7741 -1.6861>, rotation=<rot 0 0 -85>}]
 dyn_R_front_chain_2_jnt1 [{
   shapeType=Capsule
   length=6.3785
   radius=3.3062
   shapeName=dyn_L_front_chain_2_jnt1_capsule
   center=<-3.5625 1.5834 -1.3462>
   rotation=<rot 0.0001 0.0001 -85>
  }]
 dyn_R_front_chain_3_jnt [{
   shapeType=Capsule
   length=10.4886
   radius=2.641
   shapeName=dyn_L_front_chain_3_jnt_capsule
   center=<-5.6465 3.6858 -0.4785>
   rotation=<rot 0 0.0001 -85>
  }]
 dyn_R_side_front_chain_2_jnt [{
   shapeType=Capsule
   length=5.6507
   radius=4.3493
   shapeName=dyn_L_side_front_chain_2_jnt_capsule
   center=<-3.7803 0.9796 -1.3542>
   rotation=<rot 0 0 -90>
  }]
 dyn_R_side_front_chain_3_jnt [{
   shapeType=Capsule
   length=6.7045
   radius=4.2955
   shapeName=dyn_L_side_front_chain_3_jnt_capsule
   center=<-4.9849 0.1054 -0.4666>
   rotation=<rot 0 -4.9999 -90>
  }]
 dyn_R_side_rear_chain_1_jnt [{shapeType=Capsule, length=6, radius=4, shapeName=dyn_L_side_rear_chain_1_jnt_capsule, center=<-3.7803 0 -2.2336>, rotation=<rot 0 0 -90>}]
 dyn_R_side_rear_chain_2_jnt [{
   shapeType=Capsule
   length=5.84
   radius=4.16
   shapeName=dyn_L_side_rear_chain_2_jnt_capsule
   center=<-3.7803 -0.0007 -0.4196>
   rotation=<rot 0 0 -90>
  }]
 dyn_R_side_rear_chain_3_jnt [{
   shapeType=Capsule
   length=8.9171
   radius=5.1305
   shapeName=dyn_L_side_rear_chain_3_jnt_capsule
   center=<-5.4277 -0.3528 1.2813>
   rotation=<rot 0 0 -100>
  }]
 dyn_back_chain_1_jnt [{shapeType=Capsule, length=10.8473, radius=3.1527, shapeName=dyn_back_chain_1_jnt_capsule, center=<2.676 0 4.4724>, rotation=<rot 0 0 90>}]
 dyn_back_chain_2_jnt [{
   shapeType=Capsule
   length=10.8551
   radius=2.8306
   shapeName=dyn_back_chain_2_jnt_capsule
   center=<5.3924 0.0396 4.0076>
   rotation=<rot 0 0 90>
  }]
 dyn_back_chain_3_jnt [{
   shapeType=Capsule
   length=10.3256
   radius=4.0513
   shapeName=dyn_back_chain_3_jnt_capsule
   center=<5.8211 -0.0552 1.9625>
   rotation=<rot 0 0 90>
  }]
 master_L_front_chain_1_jnt [{shapeType=Capsule, length=0, radius=0, shapeName=master_L_front_chain_1_jnt_capsule}]
 master_L_side_front_chain_1_jnt [{shapeType=Capsule, length=0.505, radius=0.505, shapeName=master_L_side_front_chain_1_jnt_capsule, rotation=<rot 0 -90 0>}]
 master_L_side_rear_chain_1_jnt [{shapeType=Capsule, length=0, radius=0, shapeName=master_L_side_rear_chain_1_jnt_capsule}]
 master_R_front_chain_1_jnt [{shapeType=Capsule, length=0, radius=0, shapeName=master_R_front_chain_1_jnt_capsule}]
 master_R_side_front_chain_1_jnt [{shapeType=Capsule, length=0, radius=0, shapeName=master_R_side_front_chain_1_jnt_capsule}]
 master_R_side_rear_chain_1_jnt [{shapeType=Capsule, length=0, radius=0, shapeName=master_R_side_rear_chain_1_jnt_capsule}]
 master_back_chain_1_jnt [{shapeType=Capsule, length=0.505, radius=0.505, shapeName=master_back_chain_1_jnt_capsule, center=<3.7854 0 0>, rotation=<rot 0 0 90>}]
 pelvis [{shapeType=Capsule, length=9.0696, radius=12.182, center=<-1.7087 -0.6205 0>}]
 spine_01 [{shapeType=Capsule, length=8, radius=9.5, shapeName=spine_01_capsule}]
 spine_02 [{shapeType=Capsule, length=7.5, radius=10, shapeName=spine_02_capsule}]
 thigh_l [
  {shapeType=Sphere, radius=8, shapeName=thigh_l_sphere, center=<-14.014 2.1617 -1.7829>}
  {shapeType=Capsule, length=30, radius=8, shapeName=thigh_l_capsule, center=<-12 1.1285 0.279>, rotation=<rot 85 179.9986 179.9997>}
 ]
 thigh_r [
  {shapeType=Sphere, radius=8, shapeName=thigh_l_sphere, center=<14.014 -2.1617 1.7829>}
  {shapeType=Capsule, length=30, radius=8, shapeName=thigh_l_capsule, center=<12 -1.1285 -0.279>, rotation=<rot 85 179.9986 -0.0003>}
 ]
}
constraints [
 {
  bone1Name=spine_01
  bone2Name=pelvis
  swing1LimitDegrees=45
  swing1Motion=Free
  swing2LimitDegrees=45
  swing2Motion=Free
  twistLimitDegrees=45
  twistMotion=Free
 }
 {
  bone1Name=spine_02
  bone2Name=spine_01
  swing1LimitDegrees=45
  swing1Motion=Free
  swing2LimitDegrees=45
  swing2Motion=Free
  twistLimitDegrees=45
  twistMotion=Free
 }
 {
  bone1Name=thigh_l
  bone2Name=pelvis
  swing1LimitDegrees=45
  swing1Motion=Free
  swing2LimitDegrees=45
  swing2Motion=Free
  twistLimitDegrees=45
  twistMotion=Free
 }
 {bone1Name=master_back_chain_1_jnt, bone2Name=spine_02, swing1LimitDegrees=20, swing2LimitDegrees=40, twistLimitDegrees=20}
 {bone1Name=master_L_front_chain_1_jnt, bone2Name=spine_02, swing1LimitDegrees=40, swing2LimitDegrees=60, twistLimitDegrees=20}
 {bone1Name=dyn_L_front_chain_1_jnt, bone2Name=master_L_front_chain_1_jnt, swing1LimitDegrees=40, swing2LimitDegrees=60, twistLimitDegrees=60}
 {bone1Name=dyn_L_front_chain_2_jnt1, bone2Name=dyn_L_front_chain_1_jnt, swing1LimitDegrees=30, swing2LimitDegrees=45, twistLimitDegrees=45}
 {bone1Name=dyn_L_front_chain_3_jnt, bone2Name=dyn_L_front_chain_2_jnt1, swing1LimitDegrees=30, swing2LimitDegrees=45, twistLimitDegrees=45}
 {bone1Name=master_L_side_front_chain_1_jnt, bone2Name=spine_02, swing1LimitDegrees=40, swing2LimitDegrees=60, twistLimitDegrees=20}
 {
  bone1Name=dyn_L_side_front_chain_2_jnt
  bone2Name=master_L_side_front_chain_1_jnt
  swing1LimitDegrees=40
  swing2LimitDegrees=60
  twistLimitDegrees=20
 }
 {
  bone1Name=dyn_L_side_front_chain_3_jnt
  bone2Name=dyn_L_side_front_chain_2_jnt
  swing1LimitDegrees=40
  swing2LimitDegrees=60
  twistLimitDegrees=20
 }
 {bone1Name=master_L_side_rear_chain_1_jnt, bone2Name=spine_02, swing1LimitDegrees=40, swing2LimitDegrees=60, twistLimitDegrees=20}
 {
  bone1Name=dyn_L_side_rear_chain_1_jnt
  bone2Name=master_L_side_rear_chain_1_jnt
  swing1LimitDegrees=40
  swing2LimitDegrees=60
  twistLimitDegrees=20
 }
 {
  bone1Name=dyn_L_side_rear_chain_2_jnt
  bone2Name=dyn_L_side_rear_chain_1_jnt
  swing1LimitDegrees=30
  swing2LimitDegrees=60
  twistLimitDegrees=20
 }
 {
  bone1Name=dyn_L_side_rear_chain_3_jnt
  bone2Name=dyn_L_side_rear_chain_2_jnt
  swing1LimitDegrees=30
  swing2LimitDegrees=60
  twistLimitDegrees=20
 }
 {bone1Name=master_R_front_chain_1_jnt, bone2Name=spine_02, swing1LimitDegrees=40, swing2LimitDegrees=60, twistLimitDegrees=20}
 {bone1Name=master_R_side_front_chain_1_jnt, bone2Name=spine_02, swing1LimitDegrees=40, swing2LimitDegrees=60, twistLimitDegrees=20}
 {bone1Name=master_R_side_rear_chain_1_jnt, bone2Name=spine_02, swing1LimitDegrees=40, swing2LimitDegrees=60, twistLimitDegrees=20}
 {bone1Name=dyn_back_chain_3_jnt, bone2Name=dyn_back_chain_2_jnt, swing1LimitDegrees=20, swing2LimitDegrees=40, twistLimitDegrees=20}
 {bone1Name=dyn_back_chain_2_jnt, bone2Name=dyn_back_chain_1_jnt, swing1LimitDegrees=20, swing2LimitDegrees=40, twistLimitDegrees=20}
 {bone1Name=dyn_back_chain_2_jnt, bone2Name=dyn_L_side_rear_chain_2_jnt, swing1LimitDegrees=20, swing2LimitDegrees=40, twistLimitDegrees=20}
 {
  bone1Name=dyn_L_side_rear_chain_2_jnt
  bone2Name=dyn_L_side_front_chain_2_jnt
  swing1LimitDegrees=45
  swing1Motion=Free
  swing2LimitDegrees=45
  swing2Motion=Free
  twistLimitDegrees=45
  twistMotion=Free
 }
 {
  bone1Name=dyn_L_side_front_chain_2_jnt
  bone2Name=dyn_L_front_chain_2_jnt1
  swing1LimitDegrees=45
  swing1Motion=Free
  swing2LimitDegrees=45
  swing2Motion=Free
  twistLimitDegrees=45
  twistMotion=Free
 }
 {bone1Name=dyn_back_chain_1_jnt, bone2Name=master_back_chain_1_jnt, swing1LimitDegrees=20, swing2LimitDegrees=40, twistLimitDegrees=20}
 {bone1Name=dyn_R_front_chain_1_jnt, bone2Name=master_R_front_chain_1_jnt, swing1LimitDegrees=40, swing2LimitDegrees=60, twistLimitDegrees=60}
 {
  bone1Name=dyn_R_side_front_chain_2_jnt
  bone2Name=master_R_side_front_chain_1_jnt
  swing1LimitDegrees=40
  swing2LimitDegrees=60
  twistLimitDegrees=20
 }
 {
  bone1Name=dyn_R_side_front_chain_3_jnt
  bone2Name=dyn_R_side_front_chain_2_jnt
  swing1LimitDegrees=40
  swing2LimitDegrees=60
  twistLimitDegrees=20
 }
 {
  bone1Name=dyn_R_side_rear_chain_1_jnt
  bone2Name=master_R_side_rear_chain_1_jnt
  swing1LimitDegrees=40
  swing2LimitDegrees=60
  twistLimitDegrees=20
 }
 {
  bone1Name=dyn_R_side_rear_chain_2_jnt
  bone2Name=dyn_R_side_rear_chain_1_jnt
  swing1LimitDegrees=40
  swing2LimitDegrees=60
  twistLimitDegrees=20
 }
 {
  bone1Name=dyn_R_side_rear_chain_3_jnt
  bone2Name=dyn_R_side_rear_chain_2_jnt
  swing1LimitDegrees=40
  swing2LimitDegrees=60
  twistLimitDegrees=20
 }
 {bone1Name=dyn_R_front_chain_2_jnt1, bone2Name=dyn_R_front_chain_1_jnt, swing1LimitDegrees=30, swing2LimitDegrees=45, twistLimitDegrees=45}
 {
  bone1Name=dyn_R_side_front_chain_2_jnt
  bone2Name=dyn_R_front_chain_2_jnt1
  swing1LimitDegrees=45
  swing1Motion=Free
  swing2LimitDegrees=45
  swing2Motion=Free
  twistLimitDegrees=45
  twistMotion=Free
 }
 {bone1Name=dyn_back_chain_2_jnt, bone2Name=dyn_R_side_rear_chain_2_jnt, swing1LimitDegrees=20, swing2LimitDegrees=40, twistLimitDegrees=20}
 {
  bone1Name=dyn_R_side_front_chain_2_jnt
  bone2Name=dyn_R_side_rear_chain_2_jnt
  swing1LimitDegrees=45
  swing1Motion=Free
  swing2LimitDegrees=45
  swing2Motion=Free
  twistLimitDegrees=45
  twistMotion=Free
 }
 {bone1Name=dyn_R_front_chain_3_jnt, bone2Name=dyn_R_front_chain_2_jnt1, swing1LimitDegrees=30, swing2LimitDegrees=45, twistLimitDegrees=45}
 {
  bone1Name=dyn_R_side_front_chain_3_jnt
  bone2Name=dyn_R_front_chain_3_jnt
  swing1LimitDegrees=45
  swing1Motion=Free
  swing2LimitDegrees=45
  swing2Motion=Free
  twistLimitDegrees=45
  twistMotion=Free
 }
 {
  bone1Name=dyn_L_side_front_chain_3_jnt
  bone2Name=dyn_L_front_chain_3_jnt
  swing1LimitDegrees=45
  swing1Motion=Free
  swing2LimitDegrees=45
  swing2Motion=Free
  twistLimitDegrees=45
  twistMotion=Free
 }
 {
  bone1Name=thigh_r
  bone2Name=pelvis
  swing1LimitDegrees=45
  swing1Motion=Free
  swing2LimitDegrees=45
  swing2Motion=Free
  twistLimitDegrees=45
  twistMotion=Free
 }
]
