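# Physics asset for the jacket hems of the female medium PizzaParty body cosmetic, named for skydiving use.
# Seven hem chains hang from master anchor bodies on spine_02 through mostly swing/twist-limited joints, alongside pelvis, spine and thigh capsules.
# Angularly free cross-constraints tie neighboring chains' lower segments into a band around the back and sides, open at the front.
# The right side lacks the left's middle side-front-to-side-rear link.
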PhysicsAsset /BRCosmetics/Characters/Player/Female/Medium/Bodies/F_MED_PizzaParty/Meshes/F_MED_PizzaParty_JacketHems_Skydiving_Physics
bodies {
 dyn_L_front_chain_1_jnt [{shapeType=Capsule, length=6, radius=4, shapeName=dyn_L_front_chain_1_jnt_capsule, center=<4 0 4>, rotation=<rot 0 0 90>}]
 dyn_L_front_chain_2_jnt1 [{
   shapeType=Capsule
   length=5.8785
   radius=3.8062
   shapeName=dyn_L_front_chain_2_jnt1_capsule
   center=<3.5625 -1.6258 -0.4716>
   rotation=<rot 0 0 -90>
  }]
 dyn_L_front_chain_3_jnt [{shapeType=Sphere, radius=4, shapeName=dyn_L_front_chain_3_jnt_sphere, center=<4.55 0 0>}]
 dyn_L_side_front_chain_1_jnt [{shapeType=Capsule, length=6, radius=4, shapeName=dyn_L_side_front_chain_1_jnt_capsule, center=<3.7803 0 3.2465>, rotation=<rot 0 0 90>}]
 dyn_L_side_front_chain_2_jnt [{shapeType=Capsule, length=6, radius=4, shapeName=dyn_L_side_front_chain_2_jnt_capsule, center=<3.7803 0 3.8542>, rotation=<rot 0 0 90>}]
 dyn_L_side_front_chain_3_jnt [{shapeType=Capsule, length=5, radius=5, shapeName=dyn_L_side_front_chain_3_jnt_capsule, center=<3.7803 0 1.1385>, rotation=<rot 0 0 90>}]
 dyn_L_side_rear_chain_1_jnt [{shapeType=Capsule, length=6, radius=4, shapeName=dyn_L_side_rear_chain_1_jnt_capsule, center=<3.7803 0 3.2465>, rotation=<rot 0 0 90>}]
 dyn_L_side_rear_chain_2_jnt [{shapeType=Capsule, length=6, radius=4, shapeName=dyn_L_side_rear_chain_2_jnt_capsule, center=<3.7803 0 2.7462>, rotation=<rot 0 0 90>}]
 dyn_L_side_rear_chain_3_jnt [{shapeType=Capsule, length=5, radius=5, shapeName=dyn_L_side_rear_chain_3_jnt_capsule, center=<4.5918 0 -0.8329>, rotation=<rot 0 0 90>}]
 dyn_R_front_chain_1_jnt [{shapeType=Capsule, length=6, radius=4, shapeName=dyn_L_front_chain_1_jnt_capsule, center=<-4 0 -4>, rotation=<rot 0 0 -90>}]
 dyn_R_front_chain_2_jnt1 [{
   shapeType=Capsule
   length=5.8785
   radius=3.8062
   shapeName=dyn_L_front_chain_2_jnt1_capsule
   center=<-3.5625 1.6258 0.4716>
   rotation=<rot 0 0 90>
  }]
 dyn_R_front_chain_3_jnt [{shapeType=Sphere, radius=4, shapeName=dyn_L_front_chain_3_jnt_sphere, center=<-4.55 0 0>}]
 dyn_R_side_front_chain_1_jnt [{shapeType=Capsule, length=6, radius=4, shapeName=dyn_L_side_front_chain_1_jnt_capsule, center=<-3.7803 0 -3.2465>, rotation=<rot 0 0 -90>}]
 dyn_R_side_front_chain_2_jnt [{shapeType=Capsule, length=6, radius=4, shapeName=dyn_L_side_front_chain_2_jnt_capsule, center=<-3.7803 0 -3.8542>, rotation=<rot 0 0 -90>}]
 dyn_R_side_front_chain_3_jnt [{shapeType=Capsule, length=5, radius=5, shapeName=dyn_L_side_front_chain_3_jnt_capsule, center=<-3.7803 0 -1.1385>, rotation=<rot 0 0 -90>}]
 dyn_R_side_rear_chain_1_jnt [{shapeType=Capsule, length=6, radius=4, shapeName=dyn_L_side_rear_chain_1_jnt_capsule, center=<-3.7803 0 -3.2465>, rotation=<rot 0 0 -90>}]
 dyn_R_side_rear_chain_2_jnt [{shapeType=Capsule, length=6, radius=4, shapeName=dyn_L_side_rear_chain_2_jnt_capsule, center=<-3.7803 0 -2.7462>, rotation=<rot 0 0 -90>}]
 dyn_R_side_rear_chain_3_jnt [{shapeType=Capsule, length=5, radius=5, shapeName=dyn_L_side_rear_chain_3_jnt_capsule, center=<-4.5918 0 0.8329>, rotation=<rot 0 0 -90>}]
 dyn_back_chain_2_jnt [{
   shapeType=Capsule
   length=10.0078
   radius=3.6779
   shapeName=dyn_back_chain_2_jnt_capsule
   center=<5.9155 0.0396 0.6087>
   rotation=<rot 0 0 -90>
  }]
 dyn_back_chain_3_jnt [{shapeType=Capsule, length=9.1652, radius=4.5018, shapeName=dyn_back_chain_3_jnt_capsule, center=<4.678 0 -0.5293>, rotation=<rot 0 0 -90>}]
 master_L_front_chain_1_jnt [{shapeType=Capsule, length=0, radius=0, shapeName=master_L_front_chain_1_jnt_capsule}]
 master_L_side_front_chain_1_jnt [{shapeType=Capsule, length=0, radius=0, shapeName=master_L_side_front_chain_1_jnt_capsule}]
 master_L_side_rear_chain_1_jnt [{shapeType=Capsule, length=0, radius=0, shapeName=master_L_side_rear_chain_1_jnt_capsule}]
 master_R_front_chain_1_jnt [{shapeType=Capsule, length=0, radius=0, shapeName=master_R_front_chain_1_jnt_capsule}]
 master_R_side_front_chain_1_jnt [{shapeType=Capsule, length=0, radius=0, shapeName=master_R_side_front_chain_1_jnt_capsule}]
 master_R_side_rear_chain_1_jnt [{shapeType=Capsule, length=0, radius=0, shapeName=master_R_side_rear_chain_1_jnt_capsule}]
 master_back_chain_1_jnt [{shapeType=Capsule, length=0.505, radius=0.505, shapeName=master_back_chain_1_jnt_capsule, center=<3.7854 0 0>, rotation=<rot 0 0 90>}]
 pelvis [{shapeType=Capsule, length=11.5, radius=11, center=<-3.7428 0 0>}]
 spine_01 [{shapeType=Capsule, length=11.5, radius=9.5, shapeName=spine_01_capsule}]
 spine_02 [{shapeType=Capsule, length=11.5, radius=10, shapeName=spine_02_capsule}]
 thigh_l [
  {shapeType=Capsule, length=27.5, radius=9, shapeName=thigh_l_capsule, center=<-15.5 0.6399 0>, rotation=<rot 85 179.9986 179.9997>}
  {shapeType=Capsule, length=23.5, radius=7, shapeName=thigh_l_capsule, center=<-16.8304 3.9414 7.8826>, rotation=<rot 85 179.9992 179.9999>}
 ]
 thigh_r [{shapeType=Capsule, length=30, radius=9, shapeName=thigh_r_capsule, center=<15.5 0 0>, rotation=<rot 85 179.9998 179.9999>}]
}
constraints [
 {
  bone1Name=spine_01
  bone2Name=pelvis
  swing1LimitDegrees=45
  swing1Motion=Free
  swing2LimitDegrees=45
  swing2Motion=Free
  twistLimitDegrees=45
  twistMotion=Free
 }
 {
  bone1Name=spine_02
  bone2Name=spine_01
  swing1LimitDegrees=45
  swing1Motion=Free
  swing2LimitDegrees=45
  swing2Motion=Free
  twistLimitDegrees=45
  twistMotion=Free
 }
 {
  bone1Name=thigh_l
  bone2Name=pelvis
  swing1LimitDegrees=45
  swing1Motion=Free
  swing2LimitDegrees=45
  swing2Motion=Free
  twistLimitDegrees=45
  twistMotion=Free
 }
 {
  bone1Name=thigh_r
  bone2Name=pelvis
  swing1LimitDegrees=45
  swing1Motion=Free
  swing2LimitDegrees=45
  swing2Motion=Free
  twistLimitDegrees=45
  twistMotion=Free
 }
 {bone1Name=master_back_chain_1_jnt, bone2Name=spine_02, swing1LimitDegrees=40, swing2LimitDegrees=60, twistLimitDegrees=20}
 {bone1Name=master_L_front_chain_1_jnt, bone2Name=spine_02, swing1LimitDegrees=40, swing2LimitDegrees=60, twistLimitDegrees=20}
 {bone1Name=dyn_L_front_chain_1_jnt, bone2Name=master_L_front_chain_1_jnt, swing1LimitDegrees=40, swing2LimitDegrees=60, twistLimitDegrees=60}
 {bone1Name=dyn_L_front_chain_2_jnt1, bone2Name=dyn_L_front_chain_1_jnt, swing1LimitDegrees=40, swing2LimitDegrees=60, twistLimitDegrees=60}
 {bone1Name=dyn_L_front_chain_3_jnt, bone2Name=dyn_L_front_chain_2_jnt1, swing1LimitDegrees=40, swing2LimitDegrees=60, twistLimitDegrees=60}
 {bone1Name=master_L_side_front_chain_1_jnt, bone2Name=spine_02, swing1LimitDegrees=40, swing2LimitDegrees=60, twistLimitDegrees=20}
 {
  bone1Name=dyn_L_side_front_chain_2_jnt
  bone2Name=dyn_L_side_front_chain_1_jnt
  swing1LimitDegrees=40
  swing2LimitDegrees=60
  twistLimitDegrees=20
 }
 {
  bone1Name=dyn_L_side_front_chain_3_jnt
  bone2Name=dyn_L_side_front_chain_2_jnt
  swing1LimitDegrees=40
  swing2LimitDegrees=60
  twistLimitDegrees=20
 }
 {bone1Name=master_L_side_rear_chain_1_jnt, bone2Name=spine_02, swing1LimitDegrees=40, swing2LimitDegrees=60, twistLimitDegrees=20}
 {
  bone1Name=dyn_L_side_rear_chain_1_jnt
  bone2Name=master_L_side_rear_chain_1_jnt
  swing1LimitDegrees=40
  swing2LimitDegrees=60
  twistLimitDegrees=20
 }
 {
  bone1Name=dyn_L_side_rear_chain_2_jnt
  bone2Name=dyn_L_side_rear_chain_1_jnt
  swing1LimitDegrees=40
  swing2LimitDegrees=60
  twistLimitDegrees=20
 }
 {
  bone1Name=dyn_L_side_rear_chain_3_jnt
  bone2Name=dyn_L_side_rear_chain_2_jnt
  swing1LimitDegrees=40
  swing2LimitDegrees=60
  twistLimitDegrees=20
 }
 {bone1Name=master_R_front_chain_1_jnt, bone2Name=spine_02, swing1LimitDegrees=40, swing2LimitDegrees=60, twistLimitDegrees=20}
 {bone1Name=master_R_side_front_chain_1_jnt, bone2Name=spine_02, swing1LimitDegrees=40, swing2LimitDegrees=60, twistLimitDegrees=20}
 {bone1Name=master_R_side_rear_chain_1_jnt, bone2Name=spine_02, swing1LimitDegrees=40, swing2LimitDegrees=60, twistLimitDegrees=20}
 {bone1Name=dyn_back_chain_3_jnt, bone2Name=dyn_back_chain_2_jnt, swing1LimitDegrees=45, swing2LimitDegrees=60, twistLimitDegrees=60}
 {bone1Name=dyn_back_chain_2_jnt, bone2Name=master_back_chain_1_jnt, swing1LimitDegrees=45, swing2LimitDegrees=60, twistLimitDegrees=60}
 {
  bone1Name=dyn_L_side_front_chain_3_jnt
  bone2Name=dyn_L_front_chain_3_jnt
  swing1LimitDegrees=45
  swing1Motion=Free
  swing2LimitDegrees=45
  swing2Motion=Free
  twistLimitDegrees=45
  twistMotion=Free
 }
 {
  bone1Name=dyn_L_side_front_chain_2_jnt
  bone2Name=dyn_L_front_chain_2_jnt1
  swing1LimitDegrees=45
  swing1Motion=Free
  swing2LimitDegrees=45
  swing2Motion=Free
  twistLimitDegrees=45
  twistMotion=Free
 }
 {
  bone1Name=dyn_L_side_rear_chain_3_jnt
  bone2Name=dyn_L_side_front_chain_3_jnt
  swing1LimitDegrees=45
  swing1Motion=Free
  swing2LimitDegrees=45
  swing2Motion=Free
  twistLimitDegrees=45
  twistMotion=Free
 }
 {
  bone1Name=dyn_L_side_rear_chain_2_jnt
  bone2Name=dyn_L_side_front_chain_2_jnt
  swing1LimitDegrees=45
  swing1Motion=Free
  swing2LimitDegrees=45
  swing2Motion=Free
  twistLimitDegrees=45
  twistMotion=Free
 }
 {
  bone1Name=dyn_back_chain_3_jnt
  bone2Name=dyn_L_side_rear_chain_3_jnt
  swing1LimitDegrees=45
  swing1Motion=Free
  swing2LimitDegrees=45
  swing2Motion=Free
  twistLimitDegrees=45
  twistMotion=Free
 }
 {
  bone1Name=dyn_back_chain_2_jnt
  bone2Name=dyn_L_side_rear_chain_2_jnt
  swing1LimitDegrees=45
  swing1Motion=Free
  swing2LimitDegrees=45
  swing2Motion=Free
  twistLimitDegrees=45
  twistMotion=Free
 }
 {
  bone1Name=dyn_L_side_front_chain_1_jnt
  bone2Name=master_L_side_front_chain_1_jnt
  swing1LimitDegrees=45
  swing1Motion=Free
  swing2LimitDegrees=45
  swing2Motion=Free
  twistLimitDegrees=45
  twistMotion=Free
 }
 {bone1Name=dyn_R_front_chain_1_jnt, bone2Name=master_R_front_chain_1_jnt, swing1LimitDegrees=40, swing2LimitDegrees=60, twistLimitDegrees=60}
 {bone1Name=dyn_R_front_chain_2_jnt1, bone2Name=dyn_R_front_chain_1_jnt, swing1LimitDegrees=40, swing2LimitDegrees=60, twistLimitDegrees=60}
 {bone1Name=dyn_R_front_chain_3_jnt, bone2Name=dyn_R_front_chain_2_jnt1, swing1LimitDegrees=40, swing2LimitDegrees=60, twistLimitDegrees=60}
 {
  bone1Name=dyn_R_side_front_chain_1_jnt
  bone2Name=master_R_side_front_chain_1_jnt
  swing1LimitDegrees=40
  swing2LimitDegrees=60
  twistLimitDegrees=60
 }
 {
  bone1Name=dyn_R_side_front_chain_2_jnt
  bone2Name=dyn_R_side_front_chain_1_jnt
  swing1LimitDegrees=40
  swing2LimitDegrees=60
  twistLimitDegrees=20
 }
 {
  bone1Name=dyn_R_side_front_chain_3_jnt
  bone2Name=dyn_R_side_front_chain_2_jnt
  swing1LimitDegrees=40
  swing2LimitDegrees=60
  twistLimitDegrees=20
 }
 {
  bone1Name=dyn_R_side_rear_chain_1_jnt
  bone2Name=master_R_side_rear_chain_1_jnt
  swing1LimitDegrees=40
  swing2LimitDegrees=60
  twistLimitDegrees=20
 }
 {
  bone1Name=dyn_R_side_rear_chain_2_jnt
  bone2Name=dyn_R_side_rear_chain_1_jnt
  swing1LimitDegrees=40
  swing2LimitDegrees=60
  twistLimitDegrees=20
 }
 {
  bone1Name=dyn_R_side_rear_chain_3_jnt
  bone2Name=dyn_R_side_rear_chain_2_jnt
  swing1LimitDegrees=40
  swing2LimitDegrees=60
  twistLimitDegrees=20
 }
 {
  bone1Name=dyn_R_side_front_chain_2_jnt
  bone2Name=dyn_R_front_chain_2_jnt1
  swing1LimitDegrees=45
  swing1Motion=Free
  swing2LimitDegrees=45
  swing2Motion=Free
  twistLimitDegrees=45
  twistMotion=Free
 }
 {
  bone1Name=dyn_back_chain_2_jnt
  bone2Name=dyn_R_side_rear_chain_2_jnt
  swing1LimitDegrees=45
  swing1Motion=Free
  swing2LimitDegrees=45
  swing2Motion=Free
  twistLimitDegrees=45
  twistMotion=Free
 }
 {
  bone1Name=dyn_R_side_front_chain_3_jnt
  bone2Name=dyn_R_front_chain_3_jnt
  swing1LimitDegrees=45
  swing1Motion=Free
  swing2LimitDegrees=45
  swing2Motion=Free
  twistLimitDegrees=45
  twistMotion=Free
 }
 {
  bone1Name=dyn_R_side_rear_chain_3_jnt
  bone2Name=dyn_R_side_front_chain_3_jnt
  swing1LimitDegrees=45
  swing1Motion=Free
  swing2LimitDegrees=45
  swing2Motion=Free
  twistLimitDegrees=45
  twistMotion=Free
 }
 {
  bone1Name=dyn_back_chain_3_jnt
  bone2Name=dyn_R_side_rear_chain_3_jnt
  swing1LimitDegrees=45
  swing1Motion=Free
  swing2LimitDegrees=45
  swing2Motion=Free
  twistLimitDegrees=45
  twistMotion=Free
 }
]
