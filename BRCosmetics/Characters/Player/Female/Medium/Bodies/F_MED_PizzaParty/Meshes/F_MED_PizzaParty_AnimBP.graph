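# Animation Blueprint for the F_MED_PizzaParty female cosmetic body, adding secondary motion to jacket, collar and zippers.
# It runs the copied parent pose through Control Rig, constraints and AnimDynamics zippers, then RigidBody hems (regular and skydiving variants) and collar.
# Additive bone correctives follow physics, many weighted by arm, spine and head alphas that the EventGraph derives from bone deltas.
# The EventGraph also computes state-based gravity and drives material WPO and wind parameters.
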AnimBlueprint /BRCosmetics/Characters/Player/Female/Medium/Bodies/F_MED_PizzaParty/Meshes/F_MED_PizzaParty_AnimBP
target_skeleton=/Game/Characters/Player/Female/Medium/Bodies/F_MED_PizzaParty/Meshes/F_MED_PizzaParty_Skeleton
graphs [
 AnimGraph {nodes [
   n1 = "Transform (Modify) Bone - Bone: master_R_side_rear_chain_1_jnt"!(ComponentPose, bAlphaBoolEnabled:bool=True, Translation="-2,0,0", Rotation="0,0,-10", Scale="1,1,1") -> (Pose) {
    alphaInputType=Bool
    rotationMode=BMM_Additive
    translationMode=BMM_Additive
    alphaBoolBlend {blendInTime=0.3, blendOption=ExpInOut, blendOutTime=0.3}
    boneToModify {boneName=master_R_side_rear_chain_1_jnt}
    rotation {roll=-10}
    translation {x=-2}
   }
   "Transform (Modify) Bone - Bone: master_L_side_rear_chain_1_jnt"!(ComponentPose=@n1, bAlphaBoolEnabled:bool=True, Translation="2,0,0", Rotation="0,0,-10", Scale="1,1,1") -> (Pose) {
    alphaInputType=Bool
    rotationMode=BMM_Additive
    translationMode=BMM_Additive
    alphaBoolBlend {blendInTime=0.3, blendOption=ExpInOut, blendOutTime=0.3}
    boneToModify {boneName=master_L_side_rear_chain_1_jnt}
    rotation {roll=-10}
    translation {x=2}
   }
   n2 = Get R_armRaise95_to135() -> (R_armRaise95_to135:float)
   n3 = "Transform (Modify) Bone - Bone: dyn_collarB_MSTR_R"(
    ComponentPose=Local To Component(LocalPose=Use cached pose 'RB'())
    Alpha=Get R_armRaiseAlpha() -> (R_armRaiseAlpha:float)
    Translation="1,4,5"
    Rotation="0,-50,-30"
    Scale="1,1,1"
   ) {
    rotationMode=BMM_Additive
    translationMode=BMM_Additive
    alphaBoolBlend {blendOption=Linear}
    alphaScaleBiasClamp {bInterpResult=true}
    boneToModify {boneName=dyn_collarB_MSTR_R}
    rotation {roll=-30, yaw=-50}
    translation=<1 4 5>
   }
   n4 = Get R_armRaise45_to95() -> (R_armRaise45_to95:float)
   n5 = "Transform (Modify) Bone - Bone: dyn_collar_MSTR_R"(ComponentPose=@n3, Alpha=@n4, Translation="2,-5,2", Rotation="-6.411,0,0", Scale="1,1,1") {
    rotationMode=BMM_Additive
    rotationSpace=BCS_BoneSpace
    translationMode=BMM_Additive
    translationSpace=BCS_BoneSpace
    alphaBoolBlend {blendOption=Linear}
    alphaScaleBiasClamp {bInterpResult=true}
    boneToModify {boneName=dyn_collar_MSTR_R}
    rotation {pitch=-6.411}
    translation=<2 -5 2>
   }
   n6 = "Transform (Modify) Bone - Bone: dyn_collarB_01_R"(ComponentPose=@n5, Alpha=@n4, Translation="5,5,3", Rotation="20,-30,45", Scale="1,1,1") {
    rotationMode=BMM_Additive
    translationMode=BMM_Additive
    alphaBoolBlend {blendOption=Linear}
    alphaScaleBiasClamp {bInterpResult=true}
    boneToModify {boneName=dyn_collarB_01_R}
    rotation=<rot 20 -30 45>
    translation=<5 5 3>
   }
   n7 = Get spine_rotation() -> (spine_rotation:float)
   n8 = "Transform (Modify) Bone - Bone: dyn_collarB_MSTR_R"(ComponentPose=@n6, Alpha=@n7, Translation="1,2,3", Rotation="0,0,0", Scale="1,1,1") {
    rotationMode=BMM_Additive
    translationMode=BMM_Additive
    alphaBoolBlend {blendOption=Linear}
    alphaScaleBiasClamp {bInterpResult=true}
    boneToModify {boneName=dyn_collarB_MSTR_R}
    translation=<1 2 3>
   }
   n9 = "Transform (Modify) Bone - Bone: dyn_collar_MSTR_R"(ComponentPose=@n8, Alpha=@n7, Translation="1,0,0", Rotation="0,-7,0", Scale="1,1,1") {
    rotationMode=BMM_Additive
    rotationSpace=BCS_BoneSpace
    translationMode=BMM_Additive
    translationSpace=BCS_BoneSpace
    alphaBoolBlend {blendOption=Linear}
    alphaScaleBiasClamp {bInterpResult=true}
    boneToModify {boneName=dyn_collar_MSTR_R}
    rotation {yaw=-7}
    translation {x=1}
   }
   n10 = "Transform (Modify) Bone - Bone: dyn_collarB_01_R"(ComponentPose=@n9, Alpha=@n2, Translation="0,0,0", Rotation="-30,-30,0", Scale="1,1,1") {
    rotationMode=BMM_Additive
    translationMode=BMM_Additive
    alphaBoolBlend {blendOption=Linear}
    alphaScaleBiasClamp {bInterpResult=true}
    boneToModify {boneName=dyn_collarB_01_R}
    rotation {pitch=-30, yaw=-30}
   }
   n11 = "Transform (Modify) Bone - Bone: dyn_collar_MSTR_R"(ComponentPose=@n10, Alpha=@n2, Translation="0,0,0", Rotation="-20,0,0", Scale="1,1,1") {
    rotationMode=BMM_Additive
    translationMode=BMM_Additive
    alphaBoolBlend {blendOption=Linear}
    alphaScaleBiasClamp {bInterpResult=true}
    boneToModify {boneName=dyn_collar_MSTR_R}
    rotation {pitch=-20}
   }
   n12 = Get LeftArmBendEulerYAlpha() -> (LeftArmBendEulerYAlpha:double)
   n13 = "Transform (Modify) Bone - Bone: dyn_collar_bk_MSTR_L"(ComponentPose=@n11, Alpha=@n12, Translation="0,2,3", Rotation="-14.071,14.347,-51.800", Scale="1,1,1") {
    rotationMode=BMM_Additive
    rotationSpace=BCS_BoneSpace
    translationMode=BMM_Additive
    translationSpace=BCS_BoneSpace
    alphaBoolBlend {blendInTime=0.2, blendOption=ExpInOut, blendOutTime=0.2}
    alphaScaleBiasClamp {bInterpResult=true}
    boneToModify {boneName=dyn_collar_bk_MSTR_L}
    rotation=<rot -14.071 14.347 -51.8>
    translation {y=2, z=3}
   }
   n14 = "Transform (Modify) Bone - Bone: dyn_collar_MSTR_L"(ComponentPose=@n13, Alpha=@n12, Translation="0,0,-0.5", Rotation="-20.000,0.000,-0.000", Scale="1,1,1") {
    rotationMode=BMM_Additive
    rotationSpace=BCS_BoneSpace
    translationMode=BMM_Additive
    translationSpace=BCS_BoneSpace
    alphaBoolBlend {blendInTime=0.2, blendOption=ExpInOut, blendOutTime=0.2}
    alphaScaleBiasClamp {bInterpResult=true}
    boneToModify {boneName=dyn_collar_MSTR_L}
    rotation=<rot -20 0.0001 0>
    translation {z=-0.5}
   }
   n15 = "Transform (Modify) Bone - Bone: dyn_collar_bk_MSTR"(ComponentPose=@n14, Alpha=Get HeadLookUpAlpha() -> (HeadLookUpAlpha:double), Translation="0,0,-5", Rotation="37.5,0,0", Scale="0.7,0.7,0.7") {
    rotationMode=BMM_Additive
    rotationSpace=BCS_BoneSpace
    scaleMode=BMM_Additive
    translationMode=BMM_Additive
    translationSpace=BCS_BoneSpace
    alphaBoolBlend {blendInTime=0.2, blendOption=ExpInOut, blendOutTime=0.2}
    alphaScaleBiasClamp {bInterpResult=true}
    boneToModify {boneName=dyn_collar_bk_MSTR}
    rotation {pitch=37.5}
    scale=<0.7 0.7 0.7>
    translation {z=-5}
   }
   n16 = "Transform (Modify) Bone - Bone: master_back_chain_1_jnt"(ComponentPose=@n15, Alpha=1, Translation="3,0,4", Rotation="20,0,0", Scale="1,1,1") {
    rotationMode=BMM_Additive
    rotationSpace=BCS_BoneSpace
    translationMode=BMM_Additive
    translationSpace=BCS_BoneSpace
    alphaBoolBlend {blendOption=Linear}
    boneToModify {boneName=master_back_chain_1_jnt}
    rotation {pitch=20}
    translation {x=3, z=4}
   }
   n17 = "Transform (Modify) Bone - Bone: master_back_chain_1_jnt"!(ComponentPose=@n16, bAlphaBoolEnabled:bool=True, Translation="0,0,0", Rotation="0,0,-20", Scale="1,1,1") -> (Pose) {
    alphaInputType=Bool
    rotationMode=BMM_Additive
    translationMode=BMM_Additive
    alphaBoolBlend {blendInTime=0.3, blendOption=ExpInOut, blendOutTime=0.3}
    boneToModify {boneName=master_back_chain_1_jnt}
    rotation {roll=-20}
   }
   Output Pose(Result=Component To Local(ComponentPose=@n17)) {layerGroup=None}
   n18 = Control Rig(Source=Copy Pose From Mesh() {bUseAttachedParent=true}) {
    alphaBoolBlend {blendOption=Linear}
    controlRigAssetReference {blueprintRigClass=&/F_MED_PizzaParty_ControlRig.F_MED_PizzaParty_ControlRig_C}
   }
   n19 = "Transform (Modify) Bone - Bone: attach_backpack"(
    ComponentPose=Local To Component(LocalPose=@n18)
    Alpha=Get CustomIdle() -> (CustomIdle:double)
    Translation="0,-2.2,-1"
    Rotation="0,0,-8"
    Scale="1,1,1"
   ) {
    rotationMode=BMM_Additive
    rotationSpace=BCS_BoneSpace
    translationMode=BMM_Additive
    translationSpace=BCS_BoneSpace
    alphaBoolBlend {blendOption=Linear}
    boneToModify {boneName=attach_backpack}
    rotation {roll=-8}
    translation {y=-2.2, z=-1}
   }
   n20 = "Constraint - dyn_zipperA_mstr (PPPP)"!(
    ComponentPose=@n19
    Alpha=1
    ConstraintWeights_0=0.439
    ConstraintWeights_1:float=0.216
    ConstraintWeights_2:float=0.277
    ConstraintWeights_3:float=0.068
   ) -> (Pose) {
    alphaBoolBlend {blendOption=Linear}
    boneToModify {boneName=dyn_zipperA_mstr}
    constraintSetup [{targetBone {boneName=spine_05}}, {targetBone {boneName=neck_01}}, {targetBone {boneName=neck_02}}, {targetBone {boneName=head}}]
    constraintWeights [1,1,1,1]
   }
   n21 = "Constraint - dyn_zipperB_mstr (PPP)"!(ComponentPose=@n20, Alpha=1, ConstraintWeights_0=0.045, ConstraintWeights_1:float=0.432, ConstraintWeights_2:float=0.523) -> (Pose) {
    alphaBoolBlend {blendOption=Linear}
    boneToModify {boneName=dyn_zipperB_mstr}
    constraintSetup [{targetBone {boneName=pelvis}}, {targetBone {boneName=thigh_l}}, {targetBone {boneName=thigh_twist_01_l}}]
    constraintWeights [1,1,1]
   }
   n22 = "Constraint - dyn_zipperC_mstr (PPP)"!(ComponentPose=@n21, Alpha=1, ConstraintWeights_0=0.045, ConstraintWeights_1:float=0.432, ConstraintWeights_2:float=0.523) -> (Pose) {
    alphaBoolBlend {blendOption=Linear}
    boneToModify {boneName=dyn_zipperC_mstr}
    constraintSetup [{targetBone {boneName=pelvis}}, {targetBone {boneName=thigh_r}}, {targetBone {boneName=thigh_twist_01_r}}]
    constraintWeights [1,1,1]
   }
   n23 = Anim Dynamics - Bone: dyn_zipperA!(ComponentPose=@n22, Alpha=1, GravityOverride:Vector="0,0,-200", ExternalForce="0,0,0") -> (Pose) {
    angularDampingOverride=0.9
    angularSpringConstant=250
    bAngularSpring=true
    bOverrideAngularDamping=true
    bUseGravityOverride=true
    gravityScale=0
    alphaBoolBlend {blendOption=Linear}
    boundBone {boneName=dyn_zipperA}
    componentAppliedLinearAccClamp=<2000 2000 2000>
    componentLinearAccScale=<0.3 0.3 0.3>
    componentLinearVelScale=<0.8 0.8 0.8>
    physicsBodyDefinitions [{
      boundBone {boneName=dyn_zipperA}
      boxExtents=<5 10 10>
      constraintSetup {angularLimitsMax=<30 10 0>, angularLimitsMin=<-30 -10 -30>, angularTarget=<1 0 0>}
      localJointOffset=<0 2 3>
     }]
    relativeSpaceBone {boneName=root}
   }
   n24 = Anim Dynamics - Bone: dyn_zipperB!(ComponentPose=@n23, Alpha=1, GravityOverride:Vector="0,0,-200", ExternalForce="0,0,0") -> (Pose) {
    angularDampingOverride=0.999
    angularSpringConstant=100
    bAngularSpring=true
    bOverrideAngularDamping=true
    bUseGravityOverride=true
    bUsePlanarLimit=false
    gravityScale=0
    alphaBoolBlend {blendOption=Linear}
    boundBone {boneName=dyn_zipperB}
    componentAppliedLinearAccClamp=<900 900 900>
    componentLinearAccScale=<0.25 0.25 0.25>
    componentLinearVelScale=<0.8 0.8 0.8>
    physicsBodyDefinitions [{
      boundBone {boneName=dyn_zipperB}
      boxExtents=<5 5 5>
      constraintSetup {angularLimitsMax=<0 0 -10>, angularLimitsMin=<0 0 -40>, angularTarget=<1 0 0>}
      localJointOffset=<0 2 0>
     }]
    relativeSpaceBone {boneName=pelvis}
   }
   n25 = Anim Dynamics - Bone: dyn_zipperC!(ComponentPose=@n24, Alpha=1, GravityOverride:Vector="0,0,-100", ExternalForce="0,0,0") -> (Pose) {
    angularDampingOverride=0.99
    angularSpringConstant=80
    bAngularSpring=true
    bOverrideAngularDamping=true
    bUseGravityOverride=true
    gravityScale=0
    alphaBoolBlend {blendOption=Linear}
    boundBone {boneName=dyn_zipperC}
    componentAppliedLinearAccClamp=<900 900 900>
    componentLinearAccScale=<0.25 0.25 0.25>
    componentLinearVelScale=<0.8 0.8 0.8>
    physicsBodyDefinitions [{
      boundBone {boneName=dyn_zipperC}
      boxExtents=<5 6 6>
      constraintSetup {angularLimitsMax=<0 0 0>, angularLimitsMin=<0 0 -60>, angularTarget=<1 0 0>}
      localJointOffset=<2 2 2>
     }]
    relativeSpaceBone {boneName=root}
   }
   n26 = "Transform (Modify) Bone - Bone: dyn_collar_bk_MSTR"(ComponentPose=@n25, Alpha=1, Translation="-2,0,0", Rotation="0,0,0", Scale="0.7,0.7,0.7") {
    rotationMode=BMM_Additive
    rotationSpace=BCS_BoneSpace
    scaleMode=BMM_Additive
    translationMode=BMM_Additive
    translationSpace=BCS_BoneSpace
    alphaBoolBlend {blendInTime=0.2, blendOption=ExpInOut, blendOutTime=0.2}
    alphaScaleBiasClamp {bInterpResult=true}
    boneToModify {boneName=dyn_collar_bk_MSTR}
    scale=<0.7 0.7 0.7>
    translation {x=-2}
   }
   Save cached pose 'pre_RidigBody'(Pose=Component To Local(ComponentPose=@n26)) {cacheName=pre_RidigBody}
   n27 = RigidBody!(ComponentPose=Local To Component(LocalPose=Use cached pose 'pre_RidigBody'()), Alpha, OverrideWorldGravity:Vector="0,0,-980", ExternalForce) -> (Pose) {
    bClampLinearTranslationLimitToRefPose=true
    bOverrideWorldGravity=true
    overridePhysicsAsset=&/F_MED_PizzaParty_JacketHems_Physics
    simulationSpace=BaseBoneSpace
    alphaBoolBlend {blendOption=Linear}
    baseBoneRef {boneName=spine_01}
    componentAppliedLinearAccClamp=<2000 2000 2000>
    componentLinearAccScale=<0.2 0.2 0.2>
    componentLinearVelScale=<0.2 0.2 0.2>
    overrideWorldGravity {z=-980}
    simSpaceSettings {maxAngularAcceleration=100, maxAngularVelocity=100, maxLinearAcceleration=100, maxLinearVelocity=100, worldAlpha=0.5}
   }
   n28 = Component To Local(ComponentPose=@n27)
   n29 = RigidBody(ComponentPose=Local To Component(LocalPose=Use cached pose 'pre_RidigBody'()), OverrideWorldGravity="0,0,-980") {
    bClampLinearTranslationLimitToRefPose=true
    bOverrideWorldGravity=true
    overridePhysicsAsset=&/F_MED_PizzaParty_JacketHems_Skydiving_Physics
    simulationSpace=BaseBoneSpace
    alphaBoolBlend {blendOption=Linear}
    baseBoneRef {boneName=spine_01}
    componentAppliedLinearAccClamp=<2000 2000 2000>
    componentLinearAccScale=<0.2 0.2 0.2>
    componentLinearVelScale=<0.2 0.2 0.2>
    overrideWorldGravity {z=-980}
    simSpaceSettings {maxAngularAcceleration=100, maxAngularVelocity=100, maxLinearAcceleration=100, maxLinearVelocity=100, worldAlpha=0.5}
   }
   n30 = Component To Local(ComponentPose=@n29)
   n31 = RigidBody(
    ComponentPose=Local To Component(LocalPose=Blend Poses by bool(BlendPose_0=@n30, BlendPose_1=@n28) {blendType=Linear})
    OverrideWorldGravity="0,0,-980"
   ) {
    bClampLinearTranslationLimitToRefPose=true
    bOverrideWorldGravity=true
    overridePhysicsAsset=&/F_MED_PizzaParty_Collar_Physics
    simulationSpace=BaseBoneSpace
    alphaBoolBlend {blendOption=Linear}
    alphaScaleBiasClamp {interpSpeedDecreasing=15, interpSpeedIncreasing=20}
    baseBoneRef {boneName=spine_05}
    componentAppliedLinearAccClamp=<2000 2000 2000>
    componentLinearAccScale=<0.2 0.2 0.2>
    componentLinearVelScale=<0.2 0.2 0.2>
    overrideWorldGravity {z=-980}
    simSpaceSettings {maxAngularAcceleration=50, maxAngularVelocity=50, maxLinearAcceleration=50, maxLinearVelocity=50, velocityScaleZ=0.5, worldAlpha=1}
   }
   Save cached pose 'RB'(Pose=Component To Local(ComponentPose=@n31)) {cacheName=RB}
  ]}
 EventGraph {nodes [
   n1 = Event BlueprintUpdateAnimation()
   > Set Delta Time X(Delta Time X:double=@n1.DeltaTimeX) -> (Output_Get:double)
   > n2 = Sequence!(execute) -> (then_0, then_1, then_2:Exec, then_3:Exec, then_4:Exec, then_5:Exec, then_6:Exec, then_7:Exec, then_8:Exec, then_9:Exec)
   n3 = Get Body Mesh() -> (Body Mesh:Skeletal Mesh Component Object Reference)
   > Is Valid(InputObject=@n3)
   > Set Jacket Front Side Correction(Jacket Front Side Correction:double=GetAlphaFromReferenceTransformDelta(SkeletalMeshComponent=@n3, BoneName=thigh_l, BaseName=pelvis, InputRange=(X=-20,Y=-100), TwistAxis="0,0,1")) -> (Output_Get:double)
   n4 = Get bIsSkydiveFloating()
   n5 = Get bIsSkyDiving()
   > Set SkydivingAlpha(SkydivingAlpha:double=SelectFloat(A=1, B=0, bPickA=OR Boolean(A=@n5, B=@n4))) -> (Output_Get:double)
   n6 = GetAnimInstance(self=GetOwnerSkeletalMeshComponent())
   > Is Valid(InputObject=@n6)
   n7 = Montage_GetPosition!(
    self:Anim Instance Object Reference=@n6
    Montage:Anim Montage Object Reference=/BRCosmetics/Animation/Game/MainPlayer/Emotes/PinkSpike/Emote_PinkSpike_CMF_M
   ) -> (ReturnValue:float)
   > Set Is Custom Emote Playing(Is Custom Emote Playing:double=SelectFloat(A=1, B=0, bPickA="float > float"(A=@n7, B=1.5))) -> (Output_Get:double)
   > Set Is In Front End(Is In Front End:double="To Float (Boolean)"(InBool=Get bIsInFrontEnd())) -> (Output_Get:double)
   n8 = Get bIsTargeting()
   n9 = Get Body Mesh()
   Set R_armRaiseAlpha(
    execute=@n2.then_1
    R_armRaiseAlpha:float=float * float(
     A="To Float (Boolean)"(InBool=@n8)
     B=GetAlphaFromReferenceTransformDelta(SkeletalMeshComponent=@n9, BoneName=upperarm_r, BaseName=spine_05, InputRange=(X=45,Y=135), TwistAxis="0,0,1")
    ) {is=title K2Node_PromotableOperator}
   ) -> (Output_Get:float)
   > Set R_armRaise45_to95(R_armRaise45_to95:float=GetAlphaFromReferenceTransformDelta(SkeletalMeshComponent=@n9, BoneName=upperarm_r, BaseName=spine_05, InputRange=(X=45,Y=95), TwistAxis="0,0,1")) -> (Output_Get:float)
   > Set R_armRaise95_to135(R_armRaise95_to135:float=GetAlphaFromReferenceTransformDelta(SkeletalMeshComponent=@n9, BoneName=upperarm_r, BaseName=spine_05, InputRange=(X=95,Y=135), TwistAxis="0,0,1")) -> (Output_Get:float)
   n10 = Get bIsTargeting()
   > Set spine_rotation(spine_rotation:float=float * float(
     A="To Float (Boolean)"(InBool=@n10)
     B=GetAlphaFromReferenceTransformDelta(SkeletalMeshComponent=@n9, BoneName=spine_05, BaseName=pelvis, DeltaComponentType=EulerX, InputRange=(X=-45,Y=5), TwistAxis="0,0,1")
    )) -> (Output_Get:float)
   n11 = Get spine_rotation()
   > Set arm raise and spine rotation(arm raise and spine rotation:double=float * float(A=Get R_armRaise95_to135(), B=@n11)) -> (Output_Get:double)
   n12 = Get R_armRaise45_to95()
   > Set ArmRaiseOverHeadAndSpineRotationAlpha(ArmRaiseOverHeadAndSpineRotationAlpha:double=float / float(A=float + float(A=@n11, B=@n12), B=2)) -> (Output_Get:double)
   > Set LeftLegBendEulerY_Alpha(LeftLegBendEulerY_Alpha:double=GetAlphaFromReferenceTransformDelta(SkeletalMeshComponent=@n9, BoneName=thigh_l, BaseName=pelvis, DeltaComponentType=EulerY, InputRange=(X=10,Y=30), TwistAxis="0,0,1")) -> (Output_Get:double)
   n13 = Get SkelMesh()
   > Set LeftArmBendEulerYAlpha(LeftArmBendEulerYAlpha:double=GetAlphaFromReferenceTransformDelta(SkeletalMeshComponent=@n13, BoneName=upperarm_l, BaseName=clavicle_l, DeltaComponentType=EulerY, InputRange=(X=-35,Y=20), TwistAxis="1,0,0")) -> (Output_Get:double)
   > Set RightArmBendEulerYAlpha(RightArmBendEulerYAlpha:double=GetAlphaFromReferenceTransformDelta(SkeletalMeshComponent=@n13, BoneName=upperarm_r, BaseName=clavicle_r, DeltaComponentType=EulerY, InputRange=(X=-35,Y=20), TwistAxis="1,0,0")) -> (Output_Get:double)
   > Set ADS_spineRotationAlpha(ADS_spineRotationAlpha:float=GetAlphaFromReferenceTransformDelta(SkeletalMeshComponent=@n9, BoneName=spine_05, BaseName=pelvis, DeltaComponentType=EulerX, InputRange=(X=-15,Y=-40), TwistAxis="0,0,1")) -> (Output_Get:float)
   Branch(execute=@n2.then_5, Condition=Get bIsInFrontEnd())
   > Set CustomIdle(CustomIdle:double=GetCurveValue(self=GetOwnerFrontEndAnimInstance(), CurveName=Frontend_Idle_Bypass)) -> (Output_Get:double)
   Is Valid(exec=@n2.then_8, InputObject=Get SkelMesh())
   > n14 = Gravity Override State Function DEPRECATED(CharacterPartAnimInstance=Self-Reference(), GameTimeInSeconds=GetGameTimeInSeconds(), GravityOverrideParams=Get Gravity Override Params())
   n15 = "Equal Exactly (String)"(A=@n14.CurrentStateOut, B=OnGround_Standing)
   n16 = "Equal Exactly (String)"(A=@n14.CurrentStateOut, B=OnGround_Moving)
   n17 = Get bIsFalling()
   n18 = Get bIsJumping()
   > n19 = Branch(Condition=AND Boolean(A=OR Boolean(A=@n15, B=@n16), B=OR Boolean(A=@n17, B=@n18)))
   > Set Gravity Override(Gravity Override="0,0,-475")
   Set Gravity Override(execute=@n19.else, Gravity Override=@n14)
   Set HeadLookUpAlpha(
    execute=@n2.then_3
    HeadLookUpAlpha:double=GetAlphaFromReferenceTransformDelta(SkeletalMeshComponent=@n9, BoneName=head, BaseName=spine_05, InputRange=(X=-10,Y=-70), TwistAxis="0,0,1")
   ) -> (Output_Get:double)
   > n20 = Branch(Condition=Get bIsSkyDiving())
   > Set NonSkydivingRArmRaiseAlpha(NonSkydivingRArmRaiseAlpha:double=0) -> (Output_Get:double)
   > Set SkydivingRArmRaiseAlpha(SkydivingRArmRaiseAlpha:double=1) -> (Output_Get:double)
   Set NonSkydivingRArmRaiseAlpha(execute=@n20.else, NonSkydivingRArmRaiseAlpha=Get R_armRaise45_to95())
   > Set SkydivingRArmRaiseAlpha(SkydivingRArmRaiseAlpha=0)
   Set WPO Globals Function(
    execute=@n2.then_9
    Material=Get Mat Instance Dynamic()
    Overlay Material=Get Overlay Mat Instance Dynamic()
    Skeletal Mesh=Get SkelMesh()
    "Local Noise Transform  Location"="Get Local Noise Transform  Location"() -> ("Local Noise Transform  Location":Vector)
    Local Noise Offset=Get Local Noise Offset()
   )
   n21 = Get Delta Time X() -> (Delta Time X:double)
   > Set WPO Regional Function!(
    execute
    Material:Material Instance Dynamic Object Reference=Get Mat Instance Dynamic()
    Overlay Material:Material Instance Dynamic Object Reference=Get Overlay Mat Instance Dynamic()
    Skeletal Mesh:Skeletal Mesh Component Object Reference=Get SkelMesh()
    Velocity Transform:Name=hand_l
    Velocity Interp Speed:double=Get Velocity Interp Speed()
    Collision Transform:Name=lowerarm_l
    Collision Transform Axis:EAxis Enum
    Invert Collision Transform Axis:bool=false
    Delta Time X:double=@n21
    Previous Velocity Transform Location:"Vector (by ref)"=Get R1 Velocity Transform Location()
    Previous Velocity Transform Velocity:"Vector (by ref)"=Get R1 Velocity Transform Velocity()
    Region:int=1
   ) -> (then)
   > Set WPO Regional Function(
    Material=Get Mat Instance Dynamic()
    Overlay Material=Get Overlay Mat Instance Dynamic()
    Skeletal Mesh=Get SkelMesh()
    Velocity Transform=hand_r
    Velocity Interp Speed=Get Velocity Interp Speed()
    Collision Transform=lowerarm_r
    Invert Collision Transform Axis=false
    Delta Time X=@n21
    Previous Velocity Transform Location=Get R2 Velocity Transform Location()
    Previous Velocity Transform Velocity=Get R2 Velocity Transform Velocity()
    Region=2
   )
   > Set WPO Regional Function(
    Material=Get Mat Instance Dynamic()
    Overlay Material=Get Overlay Mat Instance Dynamic()
    Skeletal Mesh=Get SkelMesh()
    Velocity Transform=dyn_back_chain_1_jnt
    Velocity Interp Speed=Get Velocity Interp Speed()
    Collision Transform=spine_3
    Invert Collision Transform Axis=false
    Delta Time X=@n21
    Previous Velocity Transform Location=Get R3 Velocity Transform Location() -> (R3 Velocity Transform Location:Vector)
    Previous Velocity Transform Velocity=Get R3 Velocity Transform Velocity() -> (R3 Velocity Transform Velocity:Vector)
    Region=3
   )
   n22 = Get bIsSkydiveFloating()
   n23 = Get Mat Instance Dynamic()
   > SetScalarParameterValue(self=@n23, ParameterName=ClothSim_WindStrength, Value=SelectFloat(A=0.9, B=0, bPickA=@n22))
   > SetScalarParameterValue(self=@n23, ParameterName=ClothSim_WindNoiseSpeed, Value=SelectFloat(A=0.5, B=0, bPickA=@n22))
   n24 = Branch(execute=@n2.then_4, Condition=Get bIsCrouching())
   > Set Crouching RArmRaise(Crouching RArmRaise:double=Get R_armRaise45_to95()) -> (Output_Get:double)
   Set Crouching RArmRaise(execute=@n24.else, Crouching RArmRaise=0)
   Event OnCharacterCustomizationLoaded()
   > Set Body Mesh(Body Mesh:Skeletal Mesh Component Object Reference=GetOwnerSkeletalMeshComponent()) -> (Output_Get:Skeletal Mesh Component Object Reference)
   > n25 = Set Skel Mesh(Skel Mesh=GetPartSkeletalMeshComponent())
   > n26 = Get Dynamic Material Instance(SkeletalMeshComponent=@n25)
   > Set Mat Instance Dynamic(Mat Instance Dynamic=@n26)
   > n27 = GetOrCreateDynamicOverlayMaterial(MeshComponent=@n25, MaterialName=None)
   > Set Overlay Mat Instance Dynamic(Overlay Mat Instance Dynamic=@n27)
   Event BlueprintInitializeAnimation()
   > Set SkelMesh(SkelMesh=GetOwnerSkeletalMeshComponent())
  ]}
]
variables [
 ADS_spineRotationAlpha {type=number}
 arm raise and spine rotation {type=number}
 ArmRaiseOverHeadAndSpineRotationAlpha {type=number}
 BackwardMovementAlpha {type=number}
 Body Mesh {type=SkeletalMeshComponent}
 Crouching RArmRaise {type=number}
 Crouching to the left {type=number}
 CrouchingIdle {type=number}
 CrouchingIdleLookAtAlpha {type=number}
 CrouchingMovingAlpha {type=number}
 CustomIdle {type=number}
 Delta Time X {type=number}
 Gravity Override {type=Vector}
 Gravity Override Params {
  type=GravityOverrideParamsStruct
  default {
   onGround_Falling {gravityOverride=<0 0 -475>, windAmplitude=<0 0 10>, windFrequency=<0 0 100>, windOffset=<0 0 100>}
   skydive_DivingDown {gravityOverride=<-980 0 0>}
   skydive_FloatingDown {gravityOverride=<0 -600 980>, windAmplitude=<100 -400 400>, windFrequency=<5 8 10>, windOffset=<0 -300 400>}
   skydive_GoingUp {gravityOverride=<0 0 0>}
  }
 }
 Gravity Override_0 {type=Vector}
 HeadLookUpAlpha {type=number}
 Is Custom Emote Playing {type=number}
 Is In Front End {type=number}
 IsStandingIdle {type=boolean}
 Jacket Front Side Correction {type=number}
 LeftArmBendEulerYAlpha {type=number}
 LeftBegBendWhenCrouchingToLeftAlpha {type=number}
 LeftLegBendEulerY_Alpha {type=number}
 Local Noise Offset {type=Vector}
 "Local Noise Transform  Location" {type=Vector}
 Mat Instance Dynamic {type=MaterialInstanceDynamic}
 MovementToTheLeftAlpha {type=number}
 MovementToTheRightAlpha {type=number}
 NonBackwardMovementAlpha {type=number}
 NonSkydivingRArmRaiseAlpha {type=number}
 Overlay Mat Instance Dynamic {type=MaterialInstanceDynamic}
 PlayerPawn {type=FortPlayerPawnAthena}
 R1 Velocity Transform Location {type=Vector}
 R1 Velocity Transform Velocity {type=Vector}
 R2 Velocity Transform Location {type=Vector}
 R2 Velocity Transform Velocity {type=Vector}
 R3 Velocity Transform Location {type=Vector}
 R3 Velocity Transform Velocity {type=Vector}
 R_armRaise45_to95 {type=number}
 R_armRaise95_to135 {type=number}
 R_armRaiseAlpha {type=number}
 RightArmBendEulerYAlpha {type=number}
 Skel Mesh {type=SkeletalMeshComponent}
 SkelMesh {type=SkeletalMeshComponent}
 SkydivingAlpha {type=number}
 SkydivingRArmRaiseAlpha {type=number}
 spine_rotation {type=number}
 Velocity Interp Speed {type=number}
 WPO Pho Wind Intensity {type=number}
 WPO_Pho_WindNoiseSpeed {type=number}
]
