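# Post-process-style Anim Blueprint for the PizzaParty female face accessory that adds secondary motion to hair bones.
# It copies the attached parent's pose, runs back-hair Anim Dynamics and corrective bone rotation/scale, then caches it.
# Left, right and front strand dynamics chain from that cache to the output.
# Updates compute hair alphas from head-to-spine_05 angles and movement states, plus a state-based gravity override via a deprecated function.
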AnimBlueprint /BRCosmetics/Characters/Player/Female/Medium/Bodies/F_MED_PizzaParty/Meshes/Parts/F_MED_PizzaParty_FaceAcc_AnimBP
target_skeleton=&/F_MED_PizzaParty_FaceAcc_Skeleton
graphs [
 AnimGraph {nodes [
   n1 = Anim Dynamics - Bone: dyn_hair_left_strand_B_01(ComponentPose=Local To Component(LocalPose=Use cached pose 'PostBackHairDynamics'()), Alpha=0.8, ExternalForce="0,0,0") {
    angularDampingOverride=1
    angularSpringConstant=1500
    bAngularSpring=true
    bGravityOverrideInSimSpace=true
    bOverrideAngularDamping=true
    gravityScale=0.35
    simulationSpace=RootRelative
    alphaBoolBlend {blendOption=Linear}
    boundBone {boneName=dyn_hair_left_strand_B_01}
    componentAppliedLinearAccClamp=<3000 3000 3000>
    componentLinearAccScale=<0.3 0.3 0.3>
    componentLinearVelScale=<0.3 0.3 0.3>
    physicsBodyDefinitions [{
      sphereCollisionRadius=0
      boundBone {boneName=dyn_hair_left_strand_B_01}
      boxExtents=<10 15 15>
      constraintSetup {angularLimitsMax=<5 -5 50>, angularLimitsMin=<-5 -20 -35>, angularTarget=<0 0 0>}
      localJointOffset=<3 0 0>
     }]
   }
   n2 = Anim Dynamics - Bone: dyn_hair_left_strand_A_01(ComponentPose=@n1, Alpha=1, ExternalForce="0,0,0") {
    angularDampingOverride=1
    angularSpringConstant=1500
    bAngularSpring=true
    bGravityOverrideInSimSpace=true
    bOverrideAngularDamping=true
    gravityScale=0.35
    simulationSpace=RootRelative
    alphaBoolBlend {blendOption=Linear}
    boundBone {boneName=dyn_hair_left_strand_A_01}
    componentAppliedLinearAccClamp=<3000 3000 3000>
    componentLinearAccScale=<0.3 0.3 0.3>
    componentLinearVelScale=<0.3 0.3 0.3>
    physicsBodyDefinitions [{
      sphereCollisionRadius=0
      boundBone {boneName=dyn_hair_left_strand_A_01}
      boxExtents=<10 15 15>
      constraintSetup {angularLimitsMax=<5 -5 20>, angularLimitsMin=<-5 -20 -25>, angularTarget=<0 0 0>}
      localJointOffset=<3 0 0>
     }]
   }
   n3 = Anim Dynamics - Bone: dyn_hair_left_strand_A_02(ComponentPose=@n2, Alpha=1, ExternalForce="0,0,0") {
    angularDampingOverride=1
    angularSpringConstant=1500
    bAngularSpring=true
    bGravityOverrideInSimSpace=true
    bOverrideAngularDamping=true
    gravityScale=0.35
    simulationSpace=RootRelative
    alphaBoolBlend {blendOption=Linear}
    boundBone {boneName=dyn_hair_left_strand_A_02}
    componentAppliedLinearAccClamp=<3000 3000 3000>
    componentLinearAccScale=<0.3 0.3 0.3>
    componentLinearVelScale=<0.3 0.3 0.3>
    physicsBodyDefinitions [{
      sphereCollisionRadius=0
      boundBone {boneName=dyn_hair_left_strand_A_02}
      boxExtents=<10 15 15>
      constraintSetup {angularLimitsMax=<5 -5 20>, angularLimitsMin=<-5 -20 -25>, angularTarget=<0 0 0>}
      localJointOffset=<3 0 0>
     }]
   }
   n4 = "Transform (Modify) Bone - Bone: dyn_hair_left_strand_A_01"(ComponentPose=@n3, Alpha=1, Translation="0,0,0", Rotation="10.000,0.000,-0.000", Scale="1,1,1") {
    rotationMode=BMM_Additive
    rotationSpace=BCS_BoneSpace
    alphaBoolBlend {blendOption=Linear}
    boneToModify {boneName=dyn_hair_left_strand_A_01}
    rotation=<rot 10 0.0001 0>
   }
   n5 = "Transform (Modify) Bone - Bone: dyn_hair_left_strand_A_02"(ComponentPose=@n4, Alpha=1, Translation="0,0,0", Rotation="5,0,0", Scale="1,1,1") {
    rotationMode=BMM_Additive
    rotationSpace=BCS_BoneSpace
    alphaBoolBlend {blendOption=Linear}
    boneToModify {boneName=dyn_hair_left_strand_A_02}
    rotation {pitch=5}
   }
   n6 = Anim Dynamics - Bone: dyn_hair_right_strand_A_01(ComponentPose=@n5, Alpha=0.7, ExternalForce="0,0,0") {
    angularDampingOverride=1
    angularSpringConstant=1500
    bOverrideAngularDamping=true
    bUseGravityOverride=true
    gravityScale=0.35
    simulationSpace=RootRelative
    alphaBoolBlend {blendOption=Linear}
    boundBone {boneName=dyn_hair_right_strand_A_01}
    componentAppliedLinearAccClamp=<30000 30000 30000>
    componentLinearAccScale=<0.2 0.2 0.2>
    componentLinearVelScale=<0.2 0.2 0.2>
    gravityOverride {z=-980}
    physicsBodyDefinitions [{
      sphereCollisionRadius=0
      boundBone {boneName=dyn_hair_right_strand_A_01}
      boxExtents=<10 15 15>
      constraintSetup {angularTargetAxis=AxisY, angularLimitsMax=<0 -10 40>, angularLimitsMin=<0 -25 -40>, angularTarget=<0 -10 0>}
      localJointOffset=<-3 0 0>
     }]
   }
   n7 = Anim Dynamics - Bone: dyn_hair_right_strand_A_02(ComponentPose=@n6, Alpha=0.7, ExternalForce="0,0,0") {
    angularDampingOverride=0.9
    angularSpringConstant=1500
    bOverrideAngularDamping=true
    bUseGravityOverride=true
    gravityScale=0.35
    simulationSpace=RootRelative
    alphaBoolBlend {blendOption=Linear}
    boundBone {boneName=dyn_hair_right_strand_A_02}
    componentAppliedLinearAccClamp=<3000 3000 3000>
    componentLinearAccScale=<0.2 0.2 0.2>
    componentLinearVelScale=<0.2 0.2 0.2>
    gravityOverride {z=-980}
    physicsBodyDefinitions [{
      sphereCollisionRadius=0
      boundBone {boneName=dyn_hair_right_strand_A_02}
      boxExtents=<7 15 15>
      constraintSetup {angularTargetAxis=AxisY, angularLimitsMax=<0 0 5>, angularLimitsMin=<0 -10 -5>, angularTarget=<0 -10 0>}
      localJointOffset=<-1 0 0>
     }]
   }
   n8 = Anim Dynamics - Bone: dyn_hair_front_strand_B_01(ComponentPose=@n7, Alpha=0.7, ExternalForce="0,0,0") {
    angularDampingOverride=1
    angularSpringConstant=1500
    bAngularSpring=true
    bGravityOverrideInSimSpace=true
    bOverrideAngularDamping=true
    gravityScale=0.35
    simulationSpace=RootRelative
    alphaBoolBlend {blendOption=Linear}
    boundBone {boneName=dyn_hair_front_strand_B_01}
    componentAppliedLinearAccClamp {x=3000}
    componentLinearAccScale=<0.2 0.2 0.2>
    componentLinearVelScale=<0.2 0.2 0.2>
    physicsBodyDefinitions [{
      sphereCollisionRadius=0
      boundBone {boneName=dyn_hair_front_strand_B_01}
      boxExtents=<10 15 15>
      constraintSetup {angularLimitsMax=<5 0 30>, angularLimitsMin=<-5 -20 -30>, angularTarget=<0 0 0>}
      localJointOffset=<-3 0 0>
     }]
   }
   n9 = Anim Dynamics - Bone: dyn_hair_front_strand_C_01(ComponentPose=@n8, Alpha=0.7, ExternalForce="0,0,0") {
    angularDampingOverride=1
    angularSpringConstant=1500
    bAngularSpring=true
    bGravityOverrideInSimSpace=true
    bOverrideAngularDamping=true
    gravityScale=0.35
    simulationSpace=RootRelative
    alphaBoolBlend {blendOption=Linear}
    boundBone {boneName=dyn_hair_front_strand_C_01}
    componentAppliedLinearAccClamp {x=3000}
    componentLinearAccScale=<0.2 0.2 0.2>
    componentLinearVelScale=<0.2 0.2 0.2>
    physicsBodyDefinitions [{
      sphereCollisionRadius=0
      boundBone {boneName=dyn_hair_front_strand_C_01}
      boxExtents=<10 15 15>
      constraintSetup {angularLimitsMax=<5 0 5>, angularLimitsMin=<-5 -10 -5>, angularTarget=<0 0 0>}
      localJointOffset=<-3 0 0>
     }]
   }
   n10 = Anim Dynamics - Bone: dyn_hair_front_strand_A_01(ComponentPose=@n9, Alpha=1, ExternalForce="0,0,0") {
    angularDampingOverride=0.8
    angularSpringConstant=1500
    bAngularSpring=true
    bGravityOverrideInSimSpace=true
    bOverrideAngularDamping=true
    bUseGravityOverride=true
    gravityScale=0.35
    simulationSpace=RootRelative
    alphaBoolBlend {blendOption=Linear}
    boundBone {boneName=dyn_hair_front_strand_A_01}
    componentAppliedLinearAccClamp=<100 100 100>
    componentLinearAccScale=<0.1 0.1 0.1>
    componentLinearVelScale=<0.3 0.3 0.3>
    gravityOverride {z=-980}
    physicsBodyDefinitions [{
      sphereCollisionRadius=0
      boundBone {boneName=dyn_hair_front_strand_A_01}
      boxExtents=<35 20 15>
      constraintSetup {angularLimitsMax=<0 0 0>, angularLimitsMin=<0 0 0>, angularTarget=<0 0 0>}
      localJointOffset=<0 0 0>
     }]
   }
   n11 = Anim Dynamics - Bone: dyn_hair_front_strand_A_02(ComponentPose=@n10, Alpha=1, ExternalForce="0,0,0") {
    angularDampingOverride=0.95
    angularSpringConstant=1500
    bAngularSpring=true
    bGravityOverrideInSimSpace=true
    bOverrideAngularDamping=true
    bUseGravityOverride=true
    gravityScale=0.35
    simulationSpace=RootRelative
    alphaBoolBlend {blendOption=Linear}
    boundBone {boneName=dyn_hair_front_strand_A_02}
    componentAppliedLinearAccClamp=<3000 3000 3000>
    componentLinearAccScale=<0.2 0.2 0.2>
    componentLinearVelScale=<0.45 0.45 0.45>
    gravityOverride {z=-980}
    physicsBodyDefinitions [{
      sphereCollisionRadius=0
      boundBone {boneName=dyn_hair_front_strand_A_02}
      boxExtents=<15 15 15>
      constraintSetup {angularLimitsMax=<5 0 19>, angularLimitsMin=<-5 -20 -5>, angularTarget=<0 0 0>, linearAxesMax=<0.3 0.3 0.3>, linearAxesMin=<0.3 0.3 0.3>}
      localJointOffset=<3 0 0>
     }]
   }
   n12 = Anim Dynamics - Bone: dyn_hair_front_strand_A_03(ComponentPose=@n11, Alpha=1, ExternalForce="0,0,0") {
    angularDampingOverride=0.95
    angularSpringConstant=1500
    bAngularSpring=true
    bGravityOverrideInSimSpace=true
    bOverrideAngularDamping=true
    bUseGravityOverride=true
    gravityScale=0.35
    simulationSpace=RootRelative
    alphaBoolBlend {blendOption=Linear}
    boundBone {boneName=dyn_hair_front_strand_A_03}
    componentAppliedLinearAccClamp=<3000 3000 3000>
    componentLinearAccScale=<0.2 0.2 0.2>
    componentLinearVelScale=<0.45 0.45 0.45>
    gravityOverride {z=-980}
    physicsBodyDefinitions [{
      sphereCollisionRadius=0
      boundBone {boneName=dyn_hair_front_strand_A_03}
      boxExtents=<10 15 15>
      constraintSetup {angularLimitsMax=<5 0 25>, angularLimitsMin=<-5 -20 -15>, angularTarget=<0 0 0>, linearAxesMax=<0.3 0.3 0.3>, linearAxesMin=<0.3 0.3 0.3>}
      localJointOffset=<3 0 0>
     }]
   }
   Output Pose(Result=Component To Local(ComponentPose=@n12)) {layerGroup=None}
   n13 = Anim Dynamics - Bone: dyn_hair_back_up_l!(ComponentPose=Local To Component(LocalPose=Copy Pose From Mesh() {bUseAttachedParent=true}), Alpha=0.8, GravityOverride:Vector="0,0,-1500") -> (Pose) {
    alpha=0.5
    angularDampingOverride=0.8
    angularSpringConstant=1500
    bOverrideAngularDamping=true
    bUseGravityOverride=true
    simulationSpace=RootRelative
    alphaBoolBlend {blendOption=Linear}
    boundBone {boneName=dyn_hair_back_up_l}
    componentAppliedLinearAccClamp=<2500 2500 2500>
    componentLinearAccScale=<1 1 1>
    componentLinearVelScale=<1 1 1>
    gravityOverride {z=-1500}
    physicsBodyDefinitions [{
      sphereCollisionRadius=0
      boundBone {boneName=dyn_hair_back_up_l}
      boxExtents=<30 20 20>
      constraintSetup {angularLimitsMax=<0 15 10>, angularLimitsMin=<0 -30 -10>, angularTarget=<0 0 0>}
      localJointOffset=<1.3 0 0>
     }]
    relativeSpaceBone {boneName=neck_01}
   }
   n14 = Anim Dynamics - Bone: dyn_hair_back_up_r!(ComponentPose=@n13, Alpha=0.8, GravityOverride:Vector="0,0,-1500") -> (Pose) {
    alpha=0.5
    angularDampingOverride=0.8
    angularSpringConstant=1500
    bOverrideAngularDamping=true
    bUseGravityOverride=true
    simulationSpace=RootRelative
    alphaBoolBlend {blendOption=Linear}
    boundBone {boneName=dyn_hair_back_up_r}
    componentAppliedLinearAccClamp=<2500 2500 2500>
    componentLinearAccScale=<1 1 1>
    componentLinearVelScale=<1 1 1>
    gravityOverride {z=-1500}
    physicsBodyDefinitions [{
      sphereCollisionRadius=0
      boundBone {boneName=dyn_hair_back_up_r}
      boxExtents=<30 20 20>
      constraintSetup {angularLimitsMax=<5 15 10>, angularLimitsMin=<-5 -30 -10>, angularTarget=<0 0 0>}
      localJointOffset=<-1.3 0 0>
     }]
    relativeSpaceBone {boneName=neck_01}
   }
   n15 = Anim Dynamics - Bone: dyn_hair_back!(ComponentPose=@n14, Alpha=1, GravityOverride:Vector="0,0,-1500") -> (Pose) {
    alpha=0.5
    angularDampingOverride=0.95
    angularSpringConstant=1500
    bOverrideAngularDamping=true
    bUseGravityOverride=true
    simulationSpace=RootRelative
    alphaBoolBlend {blendOption=Linear}
    boundBone {boneName=dyn_hair_back}
    componentAppliedLinearAccClamp=<2000 2000 2000>
    componentLinearAccScale=<0.9 0.9 0.9>
    componentLinearVelScale=<0.9 0.9 0.9>
    gravityOverride {z=-1500}
    physicsBodyDefinitions [{
      sphereCollisionRadius=0
      boundBone {boneName=dyn_hair_back}
      boxExtents=<30 30 30>
      constraintSetup {angularLimitsMax=<0 15 10>, angularLimitsMin=<0 -45 -10>, angularTarget=<0 0 0>}
      localJointOffset=<1.8 0 0>
     }]
    relativeSpaceBone {boneName=neck_01}
   }
   n16 = Anim Dynamics - Bone: dyn_hair_back_down_l!(ComponentPose=@n15, Alpha=1, GravityOverride:Vector="0,0,0") -> (Pose) {
    angularDampingOverride=1
    angularSpringConstant=1500
    bAngularSpring=true
    bOverrideAngularDamping=true
    bUseGravityOverride=true
    gravityScale=0.35
    simulationSpace=RootRelative
    alphaBoolBlend {blendOption=Linear}
    boundBone {boneName=dyn_hair_back_down_l}
    componentLinearAccScale=<0.3 0.3 0.3>
    componentLinearVelScale=<0.3 0.3 0.3>
    physicsBodyDefinitions [{
      sphereCollisionRadius=0
      boundBone {boneName=dyn_hair_back_down_l}
      boxExtents=<20 10 10>
      constraintSetup {angularLimitsMax=<0 0 15>, angularLimitsMin=<0 -25 -15>, angularTarget=<0 0 0>}
      localJointOffset=<2.5 0 0>
     }]
   }
   n17 = Anim Dynamics - Bone: dyn_hair_back_down_r!(ComponentPose=@n16, Alpha=1, GravityOverride:Vector="0,0,0") -> (Pose) {
    angularDampingOverride=1
    angularSpringConstant=1500
    bAngularSpring=true
    bOverrideAngularDamping=true
    bUseGravityOverride=true
    gravityScale=0.35
    simulationSpace=RootRelative
    alphaBoolBlend {blendOption=Linear}
    boundBone {boneName=dyn_hair_back_down_r}
    componentLinearAccScale=<0.3 0.3 0.3>
    componentLinearVelScale=<0.3 0.3 0.3>
    physicsBodyDefinitions [{
      sphereCollisionRadius=0
      boundBone {boneName=dyn_hair_back_down_r}
      boxExtents=<20 10 10>
      constraintSetup {angularLimitsMax=<0 0 15>, angularLimitsMin=<0 -25 -15>, angularTarget=<0 0 0>}
      localJointOffset=<-2.5 0 0>
     }]
   }
   n18 = "Transform (Modify) Bone - Bone: dyn_hair_back_up_l"(ComponentPose=@n17, Alpha=1, Translation="0,0,0", Rotation="-25,0,0", Scale="0.8,1,1") {
    rotationMode=BMM_Additive
    rotationSpace=BCS_BoneSpace
    scaleMode=BMM_Replace
    scaleSpace=BCS_BoneSpace
    alphaBoolBlend {blendOption=Linear}
    alphaScaleBiasClamp {bInterpResult=true, interpSpeedDecreasing=5, interpSpeedIncreasing=5}
    boneToModify {boneName=dyn_hair_back_up_l}
    rotation {pitch=-25}
    scale {x=0.8}
   }
   n19 = "Transform (Modify) Bone - Bone: dyn_hair_back_up_r"(ComponentPose=@n18, Alpha=1, Translation="0,0,0", Rotation="-25,0,0", Scale="0.8,1,1") {
    rotationMode=BMM_Additive
    rotationSpace=BCS_BoneSpace
    scaleMode=BMM_Replace
    scaleSpace=BCS_BoneSpace
    alphaBoolBlend {blendOption=Linear}
    alphaScaleBiasClamp {bInterpResult=true, interpSpeedDecreasing=5, interpSpeedIncreasing=5}
    boneToModify {boneName=dyn_hair_back_up_r}
    rotation {pitch=-25}
    scale {x=0.8}
   }
   n20 = "Transform (Modify) Bone - Bone: dyn_hair_back"(ComponentPose=@n19, Alpha=1, Translation="0,0,0", Rotation="-30,0,0", Scale="0.8,1,1") {
    rotationMode=BMM_Additive
    rotationSpace=BCS_BoneSpace
    scaleMode=BMM_Replace
    scaleSpace=BCS_BoneSpace
    alphaBoolBlend {blendOption=Linear}
    alphaScaleBiasClamp {bInterpResult=true, interpSpeedDecreasing=5, interpSpeedIncreasing=5}
    boneToModify {boneName=dyn_hair_back}
    rotation {pitch=-30}
    scale {x=0.8}
   }
   n21 = "Transform (Modify) Bone - Bone: dyn_hair_back"(ComponentPose=@n20, Alpha=1, Translation="0,0,0", Rotation="15,0,0", Scale="0.8,0.7,0.8") {
    rotationMode=BMM_Additive
    rotationSpace=BCS_BoneSpace
    scaleMode=BMM_Replace
    scaleSpace=BCS_BoneSpace
    translationMode=BMM_Additive
    alphaBoolBlend {blendOption=Linear}
    alphaScaleBiasClamp {bInterpResult=true, interpSpeedDecreasing=5, interpSpeedIncreasing=5}
    boneToModify {boneName=dyn_hair_back}
    rotation {pitch=15}
    scale=<0.8 0.7 0.8>
   }
   n22 = "Transform (Modify) Bone - Bone: dyn_hair_back_up_r"(ComponentPose=@n21, Alpha=1, Translation="0,0,0", Rotation="55,0,0", Scale="0.6,0.7,0.8") {
    rotationMode=BMM_Additive
    rotationSpace=BCS_BoneSpace
    scaleMode=BMM_Replace
    scaleSpace=BCS_BoneSpace
    translationMode=BMM_Additive
    alphaBoolBlend {blendOption=Linear}
    alphaScaleBiasClamp {bInterpResult=true, interpSpeedDecreasing=5, interpSpeedIncreasing=5}
    boneToModify {boneName=dyn_hair_back_up_r}
    rotation {pitch=55}
    scale=<0.6 0.7 0.8>
   }
   Save cached pose 'PostBackHairDynamics'(Pose=Component To Local(ComponentPose=@n22)) {cacheName=PostBackHairDynamics}
  ]}
 EventGraph {nodes [
   Event BlueprintInitializeAnimation()
   > Set Skel Mesh(Skel Mesh=GetOwnerSkeletalMeshComponent())
   Event BlueprintUpdateAnimation()
   > n1 = Sequence()
   > K2Node_Composite_1 = Variables!(execute) -> (then:Exec) {is=title K2Node_Composite}
   K2Node_Composite_0 = "Gravity "(execute=@n1.then_1) {is=title K2Node_Composite}
  ]}
 "Gravity " {
  path="EventGraph.K2Node_Composite_0.Gravity "
  nodes [
   Inputs()
   > n1 = Gravity Override State Function DEPRECATED(
    CharacterPartAnimInstance=Self-Reference()
    GameTimeInSeconds=GetGameTimeInSeconds()
    GravityOverrideParams=Get Gravity Override Params()
    RBAN Params=Get RBAN Params()
   )
   > Set Gravity Override(Gravity Override=@n1)
   Outputs()
  ]
 }
 Variables {
  path=EventGraph.K2Node_Composite_1.Variables
  nodes [
   Inputs()
   > Set HeadLookUpAlpha(HeadLookUpAlpha:float=GetAlphaFromReferenceTransformDelta(SkeletalMeshComponent=Get Skel Mesh(), BoneName=head, BaseName=spine_05, InputRange=(X=0,Y=-40), TwistAxis="1,0,0")) -> (Output_Get:float)
   > Set HeadClosesToRightShoulderAlpha(HeadClosesToRightShoulderAlpha:float=GetAlphaFromReferenceTransformDelta(
     SkeletalMeshComponent=Get Skel Mesh()
     BoneName=head
     BaseName=spine_05
     DeltaComponentType=EulerY
     InputRange=(X=15,Y=40)
     TwistAxis="1,0,0"
    )) -> (Output_Get:float)
   n1 = "float > float"(A=Get HeadLookUpAlpha() -> (HeadLookUpAlpha:float), B=Get HeadClosesToRightShoulderAlpha() -> (HeadClosesToRightShoulderAlpha:float))
   n2 = Get HeadLookUpAlpha()
   n3 = Get HeadClosesToRightShoulderAlpha()
   n4 = Get bIsMoving2D()
   n5 = Get bIsSkyDiving()
   > Set BackHairScaleAlpha(BackHairScaleAlpha:float=SelectFloat(A=0, B=SelectFloat(A=@n2, B=@n3, bPickA=@n1), bPickA=OR Boolean(A=@n5, B=@n4))) -> (Output_Get:float)
   n6 = Get Skel Mesh()
   n7 = NOT Boolean(A=Get bIsSkyDiving())
   n8 = Get bIsCrouching()
   > Set CrouchedBackRightHairRotationAlpha(CrouchedBackRightHairRotationAlpha:double=SelectFloat(
     A=GetAlphaFromReferenceTransformDelta(SkeletalMeshComponent=@n6, BoneName=head, BaseName=spine_05, DeltaComponentType=EulerY, InputRange=(X=40,Y=50), TwistAxis="1,0,0")
     B=0
     bPickA=AND Boolean(A=@n7, B=@n8)
    )) -> (Output_Get:double)
   > Set SideBackHairDynamicsAlpha(SideBackHairDynamicsAlpha:double=SelectFloat(A=0.05, B=0.8, bPickA=Get bIsInFrontEnd())) -> (Output_Get:double)
   > Outputs!(then:Exec)
  ]
 }
]
variables [
 BackHairScaleAlpha {type=number}
 CrouchedBackRightHairRotationAlpha {type=number}
 Gravity Override {type=Vector}
 Gravity Override Params {type=GravityOverrideParamsStruct}
 HeadClosesToRightShoulderAlpha {type=number}
 HeadLookUpAlpha {type=number}
 RBAN Params {
  type=RBANParamsStruct
  default {
   dBNO {simSpaceSettings {
     maxAngularAcceleration=1000
     maxAngularVelocity=3
     maxLinearAcceleration=300
     maxLinearVelocity=700
     velocityScaleZ=0
     externalLinearDragV=<0 0.5 0>
    }}
   dBNO_Carried {simSpaceSettings {
     maxAngularAcceleration=1000
     maxAngularVelocity=3
     maxLinearAcceleration=300
     maxLinearVelocity=700
     velocityScaleZ=0
     externalLinearDragV=<0 0.5 0>
    }}
   emoteOrMelee {simSpaceSettings {
     maxAngularAcceleration=1000
     maxAngularVelocity=3
     maxLinearAcceleration=300
     maxLinearVelocity=700
     velocityScaleZ=0
     externalLinearDragV=<0 0.5 0>
    }}
   front_End {simSpaceSettings {
     maxAngularAcceleration=1000
     maxAngularVelocity=3
     maxLinearAcceleration=300
     maxLinearVelocity=700
     velocityScaleZ=0
     externalLinearDragV=<0 0.5 0>
    }}
   onGround_Crouch_Moving {simSpaceSettings {
     maxAngularAcceleration=1000
     maxAngularVelocity=3
     maxLinearAcceleration=300
     maxLinearVelocity=700
     velocityScaleZ=0
     externalLinearDragV=<0 0.5 0>
    }}
   onGround_Falling {simSpaceSettings {
     maxAngularAcceleration=1000
     maxAngularVelocity=3
     maxLinearAcceleration=300
     maxLinearVelocity=700
     velocityScaleZ=0
     externalLinearDragV=<0 0.5 0>
    }}
   onGround_Moving {simSpaceSettings {
     maxAngularAcceleration=1000
     maxAngularVelocity=3
     maxLinearAcceleration=300
     maxLinearVelocity=700
     velocityScaleZ=0
     externalLinearDragV=<0 0.5 0>
    }}
   onGround_Standing {simSpaceSettings {
     maxAngularAcceleration=1000
     maxAngularVelocity=3
     maxLinearAcceleration=300
     maxLinearVelocity=700
     velocityScaleZ=0
     externalLinearDragV=<0 0.5 0>
    }}
   rail_Grinding {simSpaceSettings {
     maxAngularAcceleration=1000
     maxAngularVelocity=3
     maxLinearAcceleration=300
     maxLinearVelocity=700
     velocityScaleZ=0
     worldAlpha=0.1
     externalLinearDragV=<0 0.5 0>
    }}
   riding_Moving {simSpaceSettings {
     maxAngularAcceleration=1000
     maxAngularVelocity=3
     maxLinearAcceleration=300
     maxLinearVelocity=700
     velocityScaleZ=0
     externalLinearDragV=<0 0.5 0>
    }}
   riding_Moving_Targeting {simSpaceSettings {
     maxAngularAcceleration=1000
     maxAngularVelocity=3
     maxLinearAcceleration=300
     maxLinearVelocity=700
     velocityScaleZ=0
     externalLinearDragV=<0 0.5 0>
    }}
   riding_Standing {simSpaceSettings {
     maxAngularAcceleration=1000
     maxAngularVelocity=3
     maxLinearAcceleration=300
     maxLinearVelocity=700
     velocityScaleZ=0
     externalLinearDragV=<0 0.5 0>
    }}
   skydive_DivingDown {simSpaceSettings {
     maxAngularAcceleration=1000
     maxAngularVelocity=3
     maxLinearAcceleration=300
     maxLinearVelocity=700
     velocityScaleZ=0
     externalLinearDragV=<0 0.5 0>
    }}
   skydive_FloatingDown {
    componentLinearAccScale=<0 0.5 0>
    componentLinearVelScale=<0 0.5 0>
    simSpaceSettings {
     maxAngularAcceleration=1000
     maxAngularVelocity=3
     maxLinearAcceleration=300
     maxLinearVelocity=700
     velocityScaleZ=0
     externalLinearDragV=<0 0.5 0>
    }
   }
   skydive_GoingUp {simSpaceSettings {
     maxAngularAcceleration=1000
     maxAngularVelocity=3
     maxLinearAcceleration=300
     maxLinearVelocity=700
     velocityScaleZ=0
     externalLinearDragV=<0 0.5 0>
    }}
   skydive_Parachuting {
    componentLinearAccScale=<0 0.1 0.3>
    componentLinearVelScale=<0 0.1 0.3>
    simSpaceSettings {
     maxAngularAcceleration=1000
     maxAngularVelocity=3
     maxLinearAcceleration=300
     maxLinearVelocity=700
     velocityScaleZ=0
     externalLinearDragV=<0 0.5 0>
    }
   }
   skydive_Surfing {simSpaceSettings {
     maxAngularAcceleration=1000
     maxAngularVelocity=3
     maxLinearAcceleration=300
     maxLinearVelocity=700
     velocityScaleZ=0
     externalLinearDragV=<0 0.5 0>
    }}
   swimming {simSpaceSettings {
     maxAngularAcceleration=1000
     maxAngularVelocity=3
     maxLinearAcceleration=300
     maxLinearVelocity=700
     velocityScaleZ=0
     externalLinearDragV=<0 0.5 0>
    }}
   tactical_Sprint {simSpaceSettings {
     maxAngularAcceleration=1000
     maxAngularVelocity=3
     maxLinearAcceleration=300
     maxLinearVelocity=700
     velocityScaleZ=0
     externalLinearDragV=<0 0.5 0>
    }}
   ziplining {simSpaceSettings {
     maxAngularAcceleration=1000
     maxAngularVelocity=3
     maxLinearAcceleration=300
     maxLinearVelocity=700
     velocityScaleZ=0
     externalLinearDragV=<0 0.5 0>
    }}
  }
 }
 SideBackHairDynamicsAlpha {type=number}
 Skel Mesh {type=SkeletalMeshComponent}
]
